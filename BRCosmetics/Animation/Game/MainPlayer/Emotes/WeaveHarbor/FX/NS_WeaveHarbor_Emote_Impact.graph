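# Niagara impact effect for the WeaveHarbor emote, featuring short-lived lightning arcs crawling over a static mesh.
# LightningAround burst-loops, picking two nearby triangles and bending ribbons between them via Hermite splines along their normals, keeping only outward arcs.
# Layered ribbons get stepped curvature and random jitter, with flickering sprites and lights at both ends.
# Ribbon point count scales with platform quality; variants LightningAround001 and LightningOut001 are disabled.
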NiagaraSystem /BRCosmetics/Animation/Game/MainPlayer/Emotes/WeaveHarbor/FX/NS_WeaveHarbor_Emote_Impact
dependencies {
 usedDataInterfaces [
  &NiagaraDataInterfacePlatformSet
  &NiagaraDataInterfaceCurve
  &NiagaraDataInterfaceColorCurve
  &NiagaraDataInterfaceCamera
  &NiagaraDataInterfaceCollisionQuery
  &NiagaraDataInterfaceArrayFloat3
  &NiagaraDataInterfaceVectorCurve
  &NiagaraDataInterfaceStaticMesh
 ]
 usedDynamicInputs []
 usedModules [
  &self:SystemSpawnScript.SystemScriptSource.SystemScriptGraph.NiagaraNodeAssignment_1.SetVariables_#1
  &SystemState
 ]
}
emitters {
 LightningAround {
  emitterSpawnScript [
   /SetVariables(Emitter.Particles In Ribbon:NiagaraInt32=&MakeCustomIntFromBool(
     Bool=&IsPlatformSetActive(New Platform Set=DataInterface() {propertyValues {Platforms {qualityLevelMask=-31}}})
     True INT=4
     False INT=&MakeCustomIntFromBool(
      Bool=&IsPlatformSetActive(New Platform Set=DataInterface() {propertyValues {Platforms {qualityLevelMask=-30}}})
      True INT=5
      False INT=&MakeCustomIntFromBool(
       Bool=&IsPlatformSetActive(New Platform Set=DataInterface() {propertyValues {Platforms {qualityLevelMask=-28}}})
       True INT=6
       False INT=6
      )
     )
    )) {isSetParameters=true, module=SetVariables_#2}
   /SetVariables(Emitter.NumTris:NiagaraInt32=&self:GetNumTris(Static Mesh:NiagaraDataInterfaceStaticMesh=$System.Static Mesh, Subtract1?:NiagaraBool=true)) {isSetParameters=true, module=SetVariables_#3}
  ]
  emitterUpdateScript [
   &EmitterState(
    Life Cycle Mode=::NewEnumerator1
    Loop Behavior=::NewEnumerator2
    Loop Count=24
    Loop Duration=&RandomRangeFloat(Minimum=0.04, Maximum=0.08)
    Recalculate Duration Each Loop=true
    UseLoopDelay=false
   )
   &SpawnBurst_Instantaneous(Spawn Count=$Emitter.Particles In Ribbon, Use Spawn Probability=false, UseLoopCountLimit=false)
   /SetVariables(Emitter.TriToSample:NiagaraInt32=&RandomRangeInt(Minimum=0, Maximum=$Emitter.NumTris)) {isSetParameters=true, module=SetVariables_#4}
   /SetVariables(Emitter.Rand oFFset:NiagaraInt32=&MakeCustomIntFromBool(
     Bool=&RandomBool()
     True INT=&RandomRangeInt(Minimum=-8, Maximum=-5)
     False INT=&RandomRangeInt(Minimum=5, Maximum=8)
    )) {isSetParameters=true, module=SetVariables_#5}
   /SetVariables(Emitter.IsInside:NiagaraBool=&RandomBool(Recalculate Random Each Loop=true)) {isSetParameters=true, module=SetVariables_#6}
   /SetVariables() {enabled=false, isSetParameters=true, module=SetVariables_#7}
  ]
  particleSpawnScript [
   &InitializeParticle(
    Color=<rgba 1 1 1 0.8>
    Lifetime=0.5
    Ribbon Width=25
    Ribbon Width Mode=::NewEnumerator1
    Sprite Size Mode=::NewEnumerator3
    Uniform Sprite Size=25
    UsePositionOffset=false
    Write Material Random=true
   )
   /SetVariables(Particles.GetTriToSample:NiagaraInt32=$Emitter.TriToSample) {isSetParameters=true, module=SetVariables_#8}
   /SetVariables(Particles.GetRandomOffset:NiagaraInt32=$Emitter.Rand oFFset) {isSetParameters=true, module=SetVariables_#9}
   /SetVariables(Particles.GetIsInside:NiagaraBool=$Emitter.IsInside) {isSetParameters=true, module=SetVariables_#10}
   &KillParticles(Kill Particles=$Particles.GetIsInside) {module=KillParticles001}
  ]
  particleUpdateScript [
   &ParticleState()
   &SpriteRotationRate()
   &self:ScratchModule(Time:NiagaraFloat=&Multiply_Float(A=$Engine.Time, B=720)) {module=ScratchModule}
   &PartitionParticles(
    Normalized Index In Partition Includes One=true
    Particles Per Partition:NiagaraInt32=$Emitter.Particles In Ribbon
    Partition Count Mode:ENiagara_InfiniteLoopDuration=::NewEnumerator1
    Partition Mode:ENiagara_PartitionMode=::NewEnumerator1
    Write Partition Index to Ribbon Link Order:ENiagara_AttributeSamplingApplyOutput=::NewEnumerator0
    Write Partition to Ribbon ID=::NewEnumerator0
   )
   /Niagara/Modules/Spawn/Location/StaticMeshLocation(
    FixedBarycentric=false
    MaskSection=false
    MaskSpawnGroup=false
    OffsetAlongNormal=false
    SamplePosOffset=false
    Static Mesh:NiagaraDataInterfaceStaticMesh=$System.Static Mesh
    Triangle ID:NiagaraInt32=$Particles.GetTriToSample
    Triangle Sampling Mode:ENiagara_StaticTriangleMode=::NewEnumerator2
    Write To Particle Position=false
   )
   &StaticMeshLocation(
    FixedBarycentric=false
    MaskSection=false
    MaskSpawnGroup=false
    OffsetAlongNormal=false
    SamplePosOffset=false
    Static Mesh:NiagaraDataInterfaceStaticMesh=$System.Static Mesh
    Triangle ID:NiagaraInt32=&Add_Integer(A=$Particles.GetTriToSample, B=$Particles.GetRandomOffset)
    Triangle Sampling Mode:ENiagara_StaticTriangleMode=::NewEnumerator2
    Write To Particle Position=false
   ) {module=StaticMeshLocation001}
   /Game/Effects/Niagara/Modules/Curves/FN_HermiteSpline(
    Direction End:Vector3f=&Multiply_VectorByFloat(
     Vector=$Particles.StaticMeshLocation001.SampledNormal
     Float=&FloatFromCurve(
      FloatCurve=DataInterface() {propertyValues {Curve {keys [
          {arriveTangentWeight=0.7758, leaveTangentWeight=0.7758, time=0, value=0.5}
          {
           arriveTangentWeight=0.4177
           interpMode=RCIM_Cubic
           leaveTangentWeight=0.4177
           tangentMode=RCTM_User
           tangentWeightMode=RCTWM_WeightedBoth
           time=1
           value=1
          }
         ]}}}
      CurveIndex=&self:pixelate_Float(float:NiagaraFloat=$Particles.NormalizedAge, FPS:NiagaraFloat=3)
      Scale Curve=&MakeCustomFloatFromBool(Bool=$Particles.GetIsInside, True Float=-50, False Float=200)
     )
    )
    Direction Start:Vector3f=&Multiply_VectorByFloat(
     Vector=$Particles.StaticMeshLocation.SampledNormal
     Float=&FloatFromCurve(
      FloatCurve=DataInterface() {propertyValues {Curve {keys [
          {arriveTangentWeight=0.7758, leaveTangentWeight=0.7758, time=0, value=0.5}
          {
           arriveTangentWeight=0.4177
           interpMode=RCIM_Cubic
           leaveTangentWeight=0.4177
           tangentMode=RCTM_User
           tangentWeightMode=RCTWM_WeightedBoth
           time=1
           value=1
          }
         ]}}}
      CurveIndex=&self:pixelate_Float(float:NiagaraFloat=$Particles.NormalizedAge, FPS:NiagaraFloat=3)
      Scale Curve=&MakeCustomFloatFromBool(Bool=$Particles.GetIsInside, True Float=-50, False Float=200)
     )
    )
    Position Blend:NiagaraFloat=$Particles.PartitionParticles.Normalized Index In Partition
    Position End:Vector3f=&AddVectorToPosition(
     Position=$Particles.StaticMeshLocation001.SampledPosition
     Vector=&Multiply_VectorByFloat(
      Vector=$Particles.StaticMeshLocation001.SampledNormal
      Float=&MakeCustomFloatFromBool(Bool=$Particles.GetIsInside, True Float=-10, False Float=3)
     )
    )
    Position Start:Vector3f=&AddVectorToPosition(
     Position=$Particles.StaticMeshLocation.SampledPosition
     Vector=&Multiply_VectorByFloat(
      Vector=$Particles.StaticMeshLocation.SampledNormal
      Float=&MakeCustomFloatFromBool(Bool=$Particles.GetIsInside, True Float=-10, False Float=3)
     )
    )
    Write Particles Position=true
   )
   /SetVariables(
    Particles.Color 1:LinearColor=&MultiplyLinearColor(
     Linear Color=$Particles.Initial.Color
     Scale Factor=&MakeVector4_FromLinearColor(LinearColor=&ColorFromCurve(ColorCurve=DataInterface() {propertyValues {
         AlphaCurve {keys [<at 0.2153 1>,<at 0.4592 0>]}
         BlueCurve {keys [<at 0.3038 2>,<at 0.4238 1>]}
         GreenCurve {keys [<at 0.3038 2>,<at 0.4238 1>]}
         RedCurve {keys [<at 0.3038 2>,<at 0.4238 1>]}
        }}))
    )
    Particles.Dynamic0_1:Vector4f=&MakeVector4(
     X=&FloatFromCurve(FloatCurve=DataInterface() {propertyValues {Curve {keys [
          {interpMode=RCIM_Cubic, leaveTangent=0.4795, tangentMode=RCTM_Break, time=0, value=0.85}
          {interpMode=RCIM_Cubic, tangentMode=RCTM_Break, time=0.6, value=0.95}
         ]}}})
     Y=&FloatFromCurve(FloatCurve=DataInterface() {propertyValues {Curve {keys [
          {arriveTangent=9.0725, interpMode=RCIM_Cubic, leaveTangent=9.0725, tangentMode=RCTM_User, time=0, value=0}
          {arriveTangent=-2.5249, interpMode=RCIM_Cubic, leaveTangent=-19.203, tangentMode=RCTM_Break, time=0.1408, value=0.9847}
          {arriveTangent=-0.0405, interpMode=RCIM_Cubic, leaveTangent=-0.0405, tangentMode=RCTM_User, time=0.3, value=0}
         ]}}})
     Z=0
     W=0
    )
   ) {isSetParameters=true, module=SetVariables_#11}
   /SetVariables(
    Particles.Color 2:LinearColor=&MultiplyLinearColor(
     Linear Color=$Particles.Initial.Color
     Scale Factor=&MakeVector4_FromLinearColor(LinearColor=&ColorFromCurve(ColorCurve=DataInterface() {propertyValues {
         AlphaCurve {keys [<at 0.2275 1>,<at 0.4546 0>]}
         BlueCurve {keys [<at 0.1752 0.0595>,<at 0.2662 0.0086>,<at 0.454 0.8958>]}
         GreenCurve {keys [<at 0.1752 0.0203>,<at 0.2662 0>,<at 0.454 0>]}
         RedCurve {keys [<at 0.1752 1>,<at 0.2662 0.2232>,<at 0.454 0.2341>]}
        }}))
    )
    Particles.Dynamic0_2:Vector4f=&MakeVector4(
     X=&FloatFromCurve(FloatCurve=DataInterface() {propertyValues {Curve {keys [
          {interpMode=RCIM_Cubic, leaveTangent=0.923, tangentMode=RCTM_Break, time=0, value=0.525}
          {interpMode=RCIM_Cubic, tangentMode=RCTM_Break, time=0.6, value=0.8}
         ]}}})
     Y=0
     Z=0
     W=0
    )
   ) {isSetParameters=true, module=SetVariables_#12}
   /SetVariables(
    Particles.Color 3:LinearColor=&MultiplyLinearColor(
     Linear Color=$Particles.Initial.Color
     Scale Factor=&Vector4FromCurve(Vector4Curve=DataInterface() {propertyValues {
        AlphaCurve {keys [<at 0 1>,<at 0.2894 0>]}
        BlueCurve {keys [<at 0 0.0188>,<at 0.3176 0.0771>]}
        GreenCurve {keys [<at 0 0.0188>,<at 0.3176 0.006>]}
        RedCurve {keys [<at 0 0.0188>,<at 0.3176 0.0371>]}
       }})
    )
    Particles.Dynamic0_3:Vector4f=&MakeVector4(
     X=&FloatFromCurve(FloatCurve=DataInterface() {propertyValues {Curve {keys [
          {interpMode=RCIM_Cubic, leaveTangent=0.2062, tangentMode=RCTM_Break, time=0, value=0.9}
          {interpMode=RCIM_Cubic, tangentMode=RCTM_Break, time=0.2, value=1}
         ]}}})
     Y=0
     Z=0
     W=0
    )
   ) {isSetParameters=true, module=SetVariables_#13}
   /SetVariables(
    Particles.Color 4:LinearColor=&MultiplyLinearColor(
     Linear Color=$Particles.Initial.Color
     Scale Factor=&MakeVector4_FromLinearColor(LinearColor=&ColorFromCurve(ColorCurve=DataInterface() {propertyValues {
         AlphaCurve {keys [<at 0.272 1>,<at 0.4546 0>]}
         BlueCurve {keys [<at 0.1791 0.0262>,<at 0.3068 0>,<at 0.4153 0.2521>]}
         GreenCurve {keys [<at 0.1791 0.0089>,<at 0.3068 0>,<at 0.4153 0>]}
         RedCurve {keys [<at 0.1791 0.4396>,<at 0.3068 0.1875>,<at 0.4153 0.0659>]}
        }}))
    )
    Particles.Dynamic0_4:Vector4f=&MakeVector4(
     X=&FloatFromCurve(FloatCurve=DataInterface() {propertyValues {Curve {keys [
          {interpMode=RCIM_Cubic, leaveTangent=0.923, tangentMode=RCTM_Break, time=0, value=0.8}
          {interpMode=RCIM_Cubic, tangentMode=RCTM_Break, time=0.6, value=1}
         ]}}})
     Y=0
     Z=0
     W=0
    )
   ) {isSetParameters=true, module=SetVariables_#14}
   &ScaleSpriteSize(
    Uniform Curve Index=$Particles.PartitionParticles.Normalized Index In Partition
    Uniform Curve Scale=0.7
    Uniform Curve Sprite Scale=DataInterface() {propertyValues {Curve {keys [{tangentMode=RCTM_User, time=0, value=1}, <at 0.01 0>, <at 0.99 0>, {tangentMode=RCTM_User, time=1, value=1}]}}}
   )
   &ScaleRibbonWidth(Ribbon Width Scale=&FloatFromCurve(
     FloatCurve=DataInterface() {propertyValues {Curve {keys [<at 0 0.5>,<at 0.01 1>,<at 0.99 1>,<at 1 0.5>]}}}
     CurveIndex=$Particles.PartitionParticles.Normalized Index In Partition
    ))
   &ScaleSpriteSize(
    Uniform Curve Scale=&Waveform("Waveform [1]"=::NewEnumerator12, "[1] Frequency"=500, "[1] Amplitude Min/Max"=<0.5 1>)
    Uniform Curve Sprite Scale=DataInterface() {propertyValues {Curve {keys [
        {arriveTangent=8.7353, interpMode=RCIM_Cubic, leaveTangent=8.7353, tangentMode=RCTM_User, time=0.05, value=0}
        {interpMode=RCIM_Cubic, tangentMode=RCTM_User, time=0.25, value=1}
        {arriveTangent=-12.4538, tangentMode=RCTM_User, time=0.5, value=0}
       ]}}}
   ) {module=ScaleSpriteSize001}
   /SetVariables(
    Particles.Sprite Color 1:LinearColor=<rgba 0.0076 0.0023 0.0229 1>
    Particles.SpriteSize 1:Vector2f=&Multiply_Vector2D(A=$Particles.SpriteSize, B=&Vector2DFromFloat(Value=1.3))
   ) {isSetParameters=true, module=SetVariables_#15}
   /SetVariables(
    Particles.Sprite Color 2:LinearColor=&MultiplyLinearColor(
     Linear Color=<rgba 0.2 0 0 1>
     Scale Factor=&MakeFloat4FromVectorAndFloat(
      "Vector (RGB)"=&VectorFromFloat(Value=1)
      "Float (Alpha)"=&FloatFromCurve(FloatCurve=DataInterface() {propertyValues {Curve {keys [<at 0.2 1>,<at 0.4 0>]}}})
     )
    )
    Particles.SpriteSize 2:Vector2f=&Multiply_Vector2D(A=$Particles.SpriteSize, B=&Vector2DFromFloat(Value=1))
   ) {isSetParameters=true, module=SetVariables_#16}
   /SetVariables(
    Particles.Sprite Color 3:LinearColor=<rgba 2.376 1.1885 3 1>
    Particles.SpriteSize 3:Vector2f=&Multiply_Vector2D(A=$Particles.SpriteSize, B=&Vector2DFromFloat(Value=0.424))
   ) {isSetParameters=true, module=SetVariables_#17}
   &FN_Light_Attributes(
    Color=&ColorFromCurve(ColorCurve=DataInterface() {propertyValues {
       AlphaCurve {keys [<at 0 1>,<at 0.7612 0>]}
       BlueCurve {keys [<at 0 0.0595>,<at 0.4522 0.8958>]}
       GreenCurve {keys [<at 0 0.0203>,<at 0.4522 0>]}
       RedCurve {keys [<at 0 1>,<at 0.4522 0.2341>]}
      }})
    Create Color Attribute=true
    Create Exponent Attribute=true
    Create Radius Attribute=true
    Enable Particle Light Enable=true
    Exponent=3
    Light Enabled:NiagaraBool=&SetBoolByFloatComparison(
     A=&FloatFromCurve(
      FloatCurve=DataInterface() {propertyValues {Curve {keys [{tangentMode=RCTM_User, time=0, value=1}, <at 0.01 0>, <at 0.99 0>, {tangentMode=RCTM_User, time=1, value=1}]}}}
      CurveIndex=$Particles.PartitionParticles.Normalized Index In Partition
     )
     B=0.5
    )
    Radius=27
    Scale Color=5
   )
   &KillParticles(Kill Particles=&SetBoolByFloatComparison(A=$Particles.NormalizedAge, B=0.5))
   &CameraOffset(Camera Offset Amount=&Lerp_Float(
     A=10
     B=-10
     Alpha=&FloatFromCurve(
      FloatCurve=DataInterface() {propertyValues {Curve {keys [{tangentMode=RCTM_User, time=-0.1, value=0}, {tangentMode=RCTM_User, time=0.1, value=1}]}}}
      CurveIndex=&DotProduct(
       Vector A=&NormalizeVector(
        Vector To Normalize=&Multiply_Vector(
         A=<1 1 0>
         B=&NormalizeVector(
          Vector To Normalize=&SubtractVectorFromPosition(
           Position=&ReturnCameraPosition(Coordinate Space:ENiagaraCoordinateSpace=::Simulation, Camera Query:NiagaraDataInterfaceCamera=DataInterface())
           Vector=/Niagara/DynamicInputs/Helpers/SimulationPosition()
          )
          Safe Normalize=true
          Safe Fallback Threshold=0
          Safe Fallback Vector=<0 0 0.001>
         )
        )
        Safe Normalize=true
        Safe Fallback Threshold=0
        Safe Fallback Vector=<0 0 0.001>
       )
       Vector B=&Multiply_VectorByFloat(
        Vector=$Particles.StaticMeshLocation.SampledNormal
        Float=&MakeCustomFloatFromBool(Bool=$Emitter.IsInside, True Float=1, False Float=1)
       )
      )
     )
    )) {enabled=false}
   &self:ScratchModule_01(
    Offset:Vector3f=&RandomVector(Vector Scale=&Multiply_Float(
      A=&FloatFromCurve(
       FloatCurve=DataInterface() {propertyValues {Curve {keys [
           {arriveTangent=5.6674, interpMode=RCIM_Cubic, leaveTangent=5.6674, tangentMode=RCTM_User, time=0.05, value=0}
           {interpMode=RCIM_Cubic, tangentMode=RCTM_User, time=0.5, value=1}
           {arriveTangent=-4.7955, interpMode=RCIM_Cubic, leaveTangent=-4.7955, tangentMode=RCTM_User, time=0.95, value=0}
          ]}}}
       CurveIndex=$Particles.PartitionParticles.Normalized Index In Partition
      )
      B=&RandomRangeFloat(Minimum=0, Maximum=30, Evaluation Type=::NewEnumerator1)
     ))
    Timer Amount:NiagaraFloat=&Divide_Float(A=1, B=6)
   ) {module=ScratchModule_01}
  ]
  rendererClasses [
   &NiagaraRibbonRendererProperties
   &NiagaraSpriteRendererProperties
   &NiagaraLightRendererProperties
  ]
  renderers [
   {rendererClass=&NiagaraRibbonRendererProperties}
   {rendererClass=&NiagaraRibbonRendererProperties, rendererIndex=1}
   {rendererClass=&NiagaraRibbonRendererProperties, rendererIndex=2}
   {rendererClass=&NiagaraRibbonRendererProperties, rendererIndex=3}
   {rendererClass=&NiagaraSpriteRendererProperties, rendererIndex=4}
   {rendererClass=&NiagaraSpriteRendererProperties, rendererIndex=5}
   {rendererClass=&NiagaraSpriteRendererProperties, rendererIndex=6}
   {rendererClass=&NiagaraLightRendererProperties, rendererIndex=7}
  ]
 }
 LightningAround001 {
  bEnabled=false
  emitterSpawnScript [
   /SetVariables(Emitter.Particles In Ribbon:NiagaraInt32=&MakeCustomIntFromBool(
     Bool=&IsPlatformSetActive(New Platform Set=DataInterface() {propertyValues {Platforms {qualityLevelMask=-31}}})
     True INT=4
     False INT=&MakeCustomIntFromBool(
      Bool=&IsPlatformSetActive(New Platform Set=DataInterface() {propertyValues {Platforms {qualityLevelMask=-30}}})
      True INT=5
      False INT=&MakeCustomIntFromBool(
       Bool=&IsPlatformSetActive(New Platform Set=DataInterface() {propertyValues {Platforms {qualityLevelMask=-28}}})
       True INT=6
       False INT=6
      )
     )
    )) {isSetParameters=true, module=SetVariables_#2}
   /SetVariables(Emitter.NumTris:NiagaraInt32=&self:GetNumTris(Static Mesh:NiagaraDataInterfaceStaticMesh=$System.Static Mesh, Subtract1?:NiagaraBool=true)) {isSetParameters=true, module=SetVariables_#3}
  ]
  emitterUpdateScript [
   &EmitterState(
    Life Cycle Mode=::NewEnumerator1
    Loop Behavior=::NewEnumerator2
    Loop Count=24
    Loop Duration=&RandomRangeFloat(Minimum=0.04, Maximum=0.08)
    Recalculate Duration Each Loop=true
    UseLoopDelay=false
   )
   &SpawnBurst_Instantaneous(Spawn Count=$Emitter.Particles In Ribbon, Use Spawn Probability=false, UseLoopCountLimit=false)
   /SetVariables(Emitter.TriToSample:NiagaraInt32=&RandomRangeInt(Minimum=0, Maximum=$Emitter.NumTris)) {isSetParameters=true, module=SetVariables_#4}
   /SetVariables(Emitter.Rand oFFset:NiagaraInt32=&MakeCustomIntFromBool(
     Bool=&RandomBool()
     True INT=&RandomRangeInt(Minimum=-8, Maximum=-4)
     False INT=&RandomRangeInt(Minimum=4, Maximum=8)
    )) {isSetParameters=true, module=SetVariables_#5}
   /SetVariables(Emitter.IsInside:NiagaraBool=&RandomBool(Recalculate Random Each Loop=true)) {isSetParameters=true, module=SetVariables_#6}
   /SetVariables(Emitter.IsInside=true) {enabled=false, isSetParameters=true, module=SetVariables_#7}
  ]
  particleSpawnScript [
   &InitializeParticle(
    Color=<rgba 1 1 1 0.8>
    Lifetime=0.5
    Ribbon Width=25
    Ribbon Width Mode=::NewEnumerator1
    Sprite Size Mode=::NewEnumerator3
    Uniform Sprite Size=25
    UsePositionOffset=false
    Write Material Random=true
   )
   /SetVariables(Particles.GetTriToSample:NiagaraInt32=$Emitter.TriToSample) {isSetParameters=true, module=SetVariables_#8}
   /SetVariables(Particles.GetRandomOffset:NiagaraInt32=$Emitter.Rand oFFset) {isSetParameters=true, module=SetVariables_#9}
   /SetVariables(Particles.GetIsInside:NiagaraBool=$Emitter.IsInside) {isSetParameters=true, module=SetVariables_#10}
  ]
  particleUpdateScript [
   &ParticleState()
   &SpriteRotationRate()
   &self:ScratchModule(Time:NiagaraFloat=&Multiply_Float(A=$Engine.Time, B=720)) {module=ScratchModule}
   &PartitionParticles(
    Normalized Index In Partition Includes One=true
    Particles Per Partition:NiagaraInt32=$Emitter.Particles In Ribbon
    Partition Count Mode:ENiagara_InfiniteLoopDuration=::NewEnumerator1
    Partition Mode:ENiagara_PartitionMode=::NewEnumerator1
    Write Partition Index to Ribbon Link Order:ENiagara_AttributeSamplingApplyOutput=::NewEnumerator0
    Write Partition to Ribbon ID=::NewEnumerator0
   )
   &StaticMeshLocation(
    FixedBarycentric=false
    MaskSection=false
    MaskSpawnGroup=false
    OffsetAlongNormal=false
    SamplePosOffset=false
    Static Mesh:NiagaraDataInterfaceStaticMesh=$System.Static Mesh
    Triangle ID:NiagaraInt32=$Particles.GetTriToSample
    Triangle Sampling Mode:ENiagara_StaticTriangleMode=::NewEnumerator2
    Write To Particle Position=false
   )
   &StaticMeshLocation(
    FixedBarycentric=false
    MaskSection=false
    MaskSpawnGroup=false
    OffsetAlongNormal=false
    SamplePosOffset=false
    Static Mesh:NiagaraDataInterfaceStaticMesh=$System.Static Mesh
    Triangle ID:NiagaraInt32=&Add_Integer(A=$Particles.GetTriToSample, B=$Particles.GetRandomOffset)
    Triangle Sampling Mode:ENiagara_StaticTriangleMode=::NewEnumerator2
    Write To Particle Position=false
   ) {module=StaticMeshLocation001}
   &FN_HermiteSpline(
    Direction End:Vector3f=&Multiply_VectorByFloat(
     Vector=$Particles.StaticMeshLocation001.SampledNormal
     Float=&FloatFromCurve(
      FloatCurve=DataInterface() {propertyValues {Curve {keys [
          {arriveTangentWeight=0.7758, leaveTangentWeight=0.7758, time=0, value=0.5}
          {
           arriveTangentWeight=0.4177
           interpMode=RCIM_Cubic
           leaveTangentWeight=0.4177
           tangentMode=RCTM_User
           tangentWeightMode=RCTWM_WeightedBoth
           time=1
           value=1
          }
         ]}}}
      Scale Curve=&MakeCustomFloatFromBool(Bool=$Particles.GetIsInside, True Float=-50, False Float=200)
     )
    )
    Direction Start:Vector3f=&Multiply_VectorByFloat(
     Vector=$Particles.StaticMeshLocation.SampledNormal
     Float=&FloatFromCurve(
      FloatCurve=DataInterface() {propertyValues {Curve {keys [
          {arriveTangentWeight=0.7758, leaveTangentWeight=0.7758, time=0, value=0.5}
          {
           arriveTangentWeight=0.4177
           interpMode=RCIM_Cubic
           leaveTangentWeight=0.4177
           tangentMode=RCTM_User
           tangentWeightMode=RCTWM_WeightedBoth
           time=1
           value=1
          }
         ]}}}
      Scale Curve=&MakeCustomFloatFromBool(Bool=$Particles.GetIsInside, True Float=-50, False Float=200)
     )
    )
    Position Blend:NiagaraFloat=$Particles.PartitionParticles.Normalized Index In Partition
    Position End:Vector3f=&AddVectorToPosition(
     Position=$Particles.StaticMeshLocation001.SampledPosition
     Vector=&Multiply_VectorByFloat(
      Vector=$Particles.StaticMeshLocation001.SampledNormal
      Float=&MakeCustomFloatFromBool(Bool=$Particles.GetIsInside, True Float=-10, False Float=3)
     )
    )
    Position Start:Vector3f=&AddVectorToPosition(
     Position=$Particles.StaticMeshLocation.SampledPosition
     Vector=&Multiply_VectorByFloat(
      Vector=$Particles.StaticMeshLocation.SampledNormal
      Float=&MakeCustomFloatFromBool(Bool=$Particles.GetIsInside, True Float=-10, False Float=3)
     )
    )
    Write Particles Position=true
   )
   /SetVariables(
    Particles.Color 1:LinearColor=&MultiplyLinearColor(
     Linear Color=$Particles.Initial.Color
     Scale Factor=&MakeVector4_FromLinearColor(LinearColor=&ColorFromCurve(ColorCurve=DataInterface() {propertyValues {
         AlphaCurve {keys [<at 0 1>,<at 0.2742 0>]}
         BlueCurve {keys [<at 0 2>,<at 0.3195 1>]}
         GreenCurve {keys [<at 0 2>,<at 0.3195 1>]}
         RedCurve {keys [<at 0 2>,<at 0.3195 1>]}
        }}))
    )
    Particles.Dynamic0_1:Vector4f=&MakeVector4(
     X=&FloatFromCurve(FloatCurve=DataInterface() {propertyValues {Curve {keys [
          {interpMode=RCIM_Cubic, leaveTangent=0.4795, tangentMode=RCTM_Break, time=0, value=0.85}
          {interpMode=RCIM_Cubic, tangentMode=RCTM_Break, time=0.6, value=0.95}
         ]}}})
     Y=&FloatFromCurve(FloatCurve=DataInterface() {propertyValues {Curve {keys [
          {arriveTangent=9.0725, interpMode=RCIM_Cubic, leaveTangent=9.0725, tangentMode=RCTM_User, time=0, value=0}
          {arriveTangent=-2.5249, interpMode=RCIM_Cubic, leaveTangent=-19.203, tangentMode=RCTM_Break, time=0.1408, value=0.9847}
          {arriveTangent=-0.0405, interpMode=RCIM_Cubic, leaveTangent=-0.0405, tangentMode=RCTM_User, time=0.3, value=0}
         ]}}})
     Z=0
     W=0
    )
   ) {isSetParameters=true, module=SetVariables_#11}
   /SetVariables(
    Particles.Color 2:LinearColor=&MultiplyLinearColor(
     Linear Color=$Particles.Initial.Color
     Scale Factor=&MakeVector4_FromLinearColor(LinearColor=&ColorFromCurve(ColorCurve=DataInterface() {propertyValues {
         AlphaCurve {keys [<at 0 1>,<at 0.4546 0>]}
         BlueCurve {keys [<at 0 0.0595>,<at 0.225 0.8958>]}
         GreenCurve {keys [<at 0 0.0203>,<at 0.225 0>]}
         RedCurve {keys [<at 0 1>,<at 0.225 0.2341>]}
        }}))
    )
    Particles.Dynamic0_2:Vector4f=&MakeVector4(
     X=&FloatFromCurve(FloatCurve=DataInterface() {propertyValues {Curve {keys [
          {interpMode=RCIM_Cubic, leaveTangent=0.923, tangentMode=RCTM_Break, time=0, value=0.525}
          {interpMode=RCIM_Cubic, tangentMode=RCTM_Break, time=0.6, value=0.8}
         ]}}})
     Y=0
     Z=0
     W=0
    )
   ) {isSetParameters=true, module=SetVariables_#12}
   /SetVariables(
    Particles.Color 3:LinearColor=&MultiplyLinearColor(
     Linear Color=$Particles.Initial.Color
     Scale Factor=&Vector4FromCurve(Vector4Curve=DataInterface() {propertyValues {
        AlphaCurve {keys [<at 0 1>,<at 0.2894 0>]}
        BlueCurve {keys [<at 0 0.0188>,<at 0.3176 0.0771>]}
        GreenCurve {keys [<at 0 0.0188>,<at 0.3176 0.006>]}
        RedCurve {keys [<at 0 0.0188>,<at 0.3176 0.0371>]}
       }})
    )
    Particles.Dynamic0_3:Vector4f=&MakeVector4(
     X=&FloatFromCurve(FloatCurve=DataInterface() {propertyValues {Curve {keys [
          {interpMode=RCIM_Cubic, leaveTangent=0.2062, tangentMode=RCTM_Break, time=0, value=0.9}
          {interpMode=RCIM_Cubic, tangentMode=RCTM_Break, time=0.2, value=1}
         ]}}})
     Y=0
     Z=0
     W=0
    )
   ) {isSetParameters=true, module=SetVariables_#13}
   /SetVariables(
    Particles.Color 4:LinearColor=&MultiplyLinearColor(
     Linear Color=$Particles.Initial.Color
     Scale Factor=&MakeVector4_FromLinearColor(LinearColor=&ColorFromCurve(ColorCurve=DataInterface() {propertyValues {
         AlphaCurve {keys [<at 0 1>,<at 0.4546 0>]}
         BlueCurve {keys [<at 0 0.0262>,<at 0.1514 0.2521>]}
         GreenCurve {keys [<at 0 0.0089>,<at 0.1514 0>]}
         RedCurve {keys [<at 0 0.4396>,<at 0.1514 0.0659>]}
        }}))
    )
    Particles.Dynamic0_4:Vector4f=&MakeVector4(
     X=&FloatFromCurve(FloatCurve=DataInterface() {propertyValues {Curve {keys [
          {interpMode=RCIM_Cubic, leaveTangent=0.923, tangentMode=RCTM_Break, time=0, value=0.8}
          {interpMode=RCIM_Cubic, tangentMode=RCTM_Break, time=0.6, value=1}
         ]}}})
     Y=0
     Z=0
     W=0
    )
   ) {isSetParameters=true, module=SetVariables_#14}
   &ScaleSpriteSize(
    Uniform Curve Index=$Particles.PartitionParticles.Normalized Index In Partition
    Uniform Curve Scale=0.7
    Uniform Curve Sprite Scale=DataInterface() {propertyValues {Curve {keys [{tangentMode=RCTM_User, time=0, value=1}, <at 0.01 0>, <at 0.99 0>, {tangentMode=RCTM_User, time=1, value=1}]}}}
   )
   &ScaleRibbonWidth(Ribbon Width Scale=&FloatFromCurve(
     FloatCurve=DataInterface() {propertyValues {Curve {keys [<at 0 0.5>,<at 0.01 1>,<at 0.99 1>,<at 1 0.5>]}}}
     CurveIndex=$Particles.PartitionParticles.Normalized Index In Partition
    ))
   &ScaleSpriteSize(
    Uniform Curve Scale=&Waveform("Waveform [1]"=::NewEnumerator12, "[1] Frequency"=500, "[1] Amplitude Min/Max"=<0.5 1>)
    Uniform Curve Sprite Scale=DataInterface() {propertyValues {Curve {keys [
        {arriveTangent=8.7353, interpMode=RCIM_Cubic, leaveTangent=8.7353, tangentMode=RCTM_User, time=0.05, value=0}
        {interpMode=RCIM_Cubic, tangentMode=RCTM_User, time=0.25, value=1}
        {arriveTangent=-12.4538, tangentMode=RCTM_User, time=0.5, value=0}
       ]}}}
   ) {module=ScaleSpriteSize001}
   /SetVariables(
    Particles.Sprite Color 1:LinearColor=<rgba 0.0076 0.0023 0.0229 1>
    Particles.SpriteSize 1:Vector2f=&Multiply_Vector2D(A=$Particles.SpriteSize, B=&Vector2DFromFloat(Value=1.3))
   ) {isSetParameters=true, module=SetVariables_#15}
   /SetVariables(
    Particles.Sprite Color 2:LinearColor=<rgba 1.2 0.0243 0.0714 1>
    Particles.SpriteSize 2:Vector2f=&Multiply_Vector2D(A=$Particles.SpriteSize, B=&Vector2DFromFloat(Value=1))
   ) {isSetParameters=true, module=SetVariables_#16}
   /SetVariables(
    Particles.Sprite Color 3:LinearColor=<rgba 2.376 1.1885 3 1>
    Particles.SpriteSize 3:Vector2f=&Multiply_Vector2D(A=$Particles.SpriteSize, B=&Vector2DFromFloat(Value=0.424))
   ) {isSetParameters=true, module=SetVariables_#17}
   &FN_Light_Attributes(
    Color=&ColorFromCurve(ColorCurve=DataInterface() {propertyValues {
       AlphaCurve {keys [<at 0 1>,<at 0.7612 0>]}
       BlueCurve {keys [<at 0 0.0595>,<at 0.4522 0.8958>]}
       GreenCurve {keys [<at 0 0.0203>,<at 0.4522 0>]}
       RedCurve {keys [<at 0 1>,<at 0.4522 0.2341>]}
      }})
    Create Color Attribute=true
    Create Exponent Attribute=true
    Create Radius Attribute=true
    Enable Particle Light Enable=true
    Exponent=3
    Light Enabled:NiagaraBool=&SetBoolByFloatComparison(
     A=&FloatFromCurve(
      FloatCurve=DataInterface() {propertyValues {Curve {keys [{tangentMode=RCTM_User, time=0, value=1}, <at 0.01 0>, <at 0.99 0>, {tangentMode=RCTM_User, time=1, value=1}]}}}
      CurveIndex=$Particles.PartitionParticles.Normalized Index In Partition
     )
     B=0.5
    )
    Radius=27
    Scale Color=5
   )
   &KillParticles(Kill Particles=&SetBoolByFloatComparison(A=$Particles.NormalizedAge, B=0.5))
   &CameraOffset(Camera Offset Amount=&Lerp_Float(
     A=10
     B=-10
     Alpha=&FloatFromCurve(
      FloatCurve=DataInterface() {propertyValues {Curve {keys [{tangentMode=RCTM_User, time=-0.1, value=0}, {tangentMode=RCTM_User, time=0.1, value=1}]}}}
      CurveIndex=&DotProduct(
       Vector A=&NormalizeVector(
        Vector To Normalize=&Multiply_Vector(
         A=<1 1 0>
         B=&NormalizeVector(
          Vector To Normalize=&SubtractVectorFromPosition(
           Position=&ReturnCameraPosition(Coordinate Space:ENiagaraCoordinateSpace=::Simulation, Camera Query:NiagaraDataInterfaceCamera=DataInterface())
           Vector=&SimulationPosition()
          )
          Safe Normalize=true
          Safe Fallback Threshold=0
          Safe Fallback Vector=<0 0 0.001>
         )
        )
        Safe Normalize=true
        Safe Fallback Threshold=0
        Safe Fallback Vector=<0 0 0.001>
       )
       Vector B=&Multiply_VectorByFloat(
        Vector=$Particles.StaticMeshLocation.SampledNormal
        Float=&MakeCustomFloatFromBool(Bool=$Emitter.IsInside, True Float=1, False Float=1)
       )
      )
     )
    )) {enabled=false}
  ]
  rendererClasses [
   &NiagaraRibbonRendererProperties
   &NiagaraSpriteRendererProperties
   &NiagaraLightRendererProperties
  ]
  renderers [
   {rendererClass=&NiagaraRibbonRendererProperties}
   {rendererClass=&NiagaraRibbonRendererProperties, rendererIndex=1}
   {rendererClass=&NiagaraRibbonRendererProperties, rendererIndex=2}
   {rendererClass=&NiagaraRibbonRendererProperties, rendererIndex=3}
   {rendererClass=&NiagaraSpriteRendererProperties, rendererIndex=4}
   {rendererClass=&NiagaraSpriteRendererProperties, rendererIndex=5}
   {rendererClass=&NiagaraSpriteRendererProperties, rendererIndex=6}
   {rendererClass=&NiagaraLightRendererProperties, rendererIndex=7}
  ]
 }
 LightningOut001 {
  bEnabled=false
  emitterSpawnScript [
   /SetVariables(Emitter.Particles In Ribbon:NiagaraInt32=&MakeCustomIntFromBool(
     Bool=&IsPlatformSetActive(New Platform Set=DataInterface() {propertyValues {Platforms {qualityLevelMask=-31}}})
     True INT=4
     False INT=&MakeCustomIntFromBool(
      Bool=&IsPlatformSetActive(New Platform Set=DataInterface() {propertyValues {Platforms {qualityLevelMask=-30}}})
      True INT=5
      False INT=&MakeCustomIntFromBool(
       Bool=&IsPlatformSetActive(New Platform Set=DataInterface() {propertyValues {Platforms {qualityLevelMask=-28}}})
       True INT=6
       False INT=7
      )
     )
    )) {isSetParameters=true, module=SetVariables_#2}
   /SetVariables(Emitter.NumTris:NiagaraInt32=&self:GetNumTris(Static Mesh:NiagaraDataInterfaceStaticMesh=$System.Static Mesh, Subtract1?:NiagaraBool=true)) {isSetParameters=true, module=SetVariables_#3}
  ]
  emitterUpdateScript [
   &EmitterState(
    Life Cycle Mode=::NewEnumerator1
    Loop Behavior=::NewEnumerator2
    Loop Count=23
    Loop Duration=&RandomRangeFloat(Minimum=0.01, Maximum=0.05)
    Recalculate Duration Each Loop=true
    UseLoopDelay=false
   )
   &SpawnBurst_Instantaneous(Spawn Count=$Emitter.Particles In Ribbon, Use Spawn Probability=false, UseLoopCountLimit=false)
   /SetVariables(Emitter.TriToSample:NiagaraInt32=&RandomRangeInt(Minimum=0, Maximum=$Emitter.NumTris)) {isSetParameters=true, module=SetVariables_#4}
   /SetVariables(Emitter.Rand oFFset:NiagaraInt32=&MakeCustomIntFromBool(
     Bool=&RandomBool()
     True INT=&RandomRangeInt(Minimum=-8, Maximum=-4)
     False INT=&RandomRangeInt(Minimum=4, Maximum=8)
    )) {isSetParameters=true, module=SetVariables_#5}
   /SetVariables(Emitter.IsInside:NiagaraBool=&RandomBool(Recalculate Random Each Loop=true)) {isSetParameters=true, module=SetVariables_#7}
  ]
  particleSpawnScript [
   &InitializeParticle(
    Color=<rgba 1 1 1 0.8>
    Lifetime=0.7
    Ribbon Width=25
    Ribbon Width Mode=::NewEnumerator1
    Sprite Size Mode=::NewEnumerator3
    Uniform Sprite Size=25
    UsePositionOffset=false
    Write Material Random=true
   )
   /SetVariables(Particles.GetTriToSample:NiagaraInt32=$Emitter.TriToSample) {isSetParameters=true, module=SetVariables_#8}
   /SetVariables(Particles.GetRandomOffset:NiagaraInt32=$Emitter.Rand oFFset) {isSetParameters=true, module=SetVariables_#9}
   /SetVariables(Particles.GetIsInside:NiagaraBool=$Emitter.IsInside) {isSetParameters=true, module=SetVariables_#10}
  ]
  particleUpdateScript [
   &ParticleState()
   &SpriteRotationRate()
   &self:ScratchModule(Time:NiagaraFloat=&Multiply_Float(A=$Engine.Time, B=720)) {module=ScratchModule}
   &PartitionParticles(
    Normalized Index In Partition Includes One=true
    Particles Per Partition:NiagaraInt32=$Emitter.Particles In Ribbon
    Partition Count Mode:ENiagara_InfiniteLoopDuration=::NewEnumerator1
    Partition Mode:ENiagara_PartitionMode=::NewEnumerator1
    Write Partition Index to Ribbon Link Order:ENiagara_AttributeSamplingApplyOutput=::NewEnumerator0
    Write Partition to Ribbon ID=::NewEnumerator0
   )
   &StaticMeshLocation(
    FixedBarycentric=false
    MaskSection=false
    MaskSpawnGroup=false
    Offset Position Along Sampled Normal=10
    OffsetAlongNormal=true
    SamplePosOffset=false
    Static Mesh:NiagaraDataInterfaceStaticMesh=$System.Static Mesh
    Triangle ID:NiagaraInt32=$Particles.GetTriToSample
    Triangle Sampling Mode:ENiagara_StaticTriangleMode=::NewEnumerator2
    Write To Particle Position=false
   )
   &FN_HermiteSpline(
    Position Blend:NiagaraFloat=&Multiply_Float(
     A=$Particles.PartitionParticles.Normalized Index In Partition
     B=&FloatFromCurve(FloatCurve=DataInterface() {propertyValues {Curve {keys [
          {arriveTangent=2.2642, interpMode=RCIM_Cubic, leaveTangent=2.2642, tangentMode=RCTM_User, time=0, value=0}
          {tangentMode=RCTM_User, time=1, value=1}
         ]}}})
    )
    Position End:Vector3f=&AddVectorToPosition(
     Position=$Particles.StaticMeshLocation.SampledPosition
     Vector=&Add_Vector(
      A=&Multiply_VectorByFloat(Vector=$Particles.StaticMeshLocation.SampledNormal, Float=&RandomRangeFloat(Minimum=75, Maximum=50))
      B=<0 0 35>
     )
    )
    Position Start:Vector3f=&AddVectorToPosition(
     Position=$Particles.StaticMeshLocation.SampledPosition
     Vector=&Multiply_VectorByFloat(Vector=$Particles.StaticMeshLocation.SampledNormal, Float=4)
    )
    Write Particles Position=true
   )
   /SetVariables(
    Particles.Color 1:LinearColor=&MultiplyLinearColor(
     Linear Color=$Particles.Initial.Color
     Scale Factor=&MakeVector4_FromLinearColor(LinearColor=&ColorFromCurve(ColorCurve=DataInterface() {propertyValues {
         AlphaCurve {keys [<at 0 1 RCIM_Cubic>, {arriveTangent=-3.1173, interpMode=RCIM_Cubic, leaveTangent=-3.1173, tangentMode=RCTM_User, time=0.7808, value=0}]}
         BlueCurve {keys [<at 0 2>,<at 0.3195 1>]}
         GreenCurve {keys [<at 0 2>,<at 0.3195 1>]}
         RedCurve {keys [<at 0 2>,<at 0.3195 1>]}
        }}))
    )
    Particles.Dynamic0_1:Vector4f=&MakeVector4(
     X=&FloatFromCurve(FloatCurve=DataInterface() {propertyValues {Curve {keys [
          {interpMode=RCIM_Cubic, leaveTangent=0.4795, tangentMode=RCTM_Break, time=0, value=0.85}
          {interpMode=RCIM_Cubic, tangentMode=RCTM_Break, time=0.6, value=0.95}
         ]}}})
     Y=&FloatFromCurve(FloatCurve=DataInterface() {propertyValues {Curve {keys [
          {arriveTangent=9.0725, interpMode=RCIM_Cubic, leaveTangent=9.0725, tangentMode=RCTM_User, time=0, value=0}
          {arriveTangent=-2.5249, interpMode=RCIM_Cubic, leaveTangent=-19.203, tangentMode=RCTM_Break, time=0.1408, value=0.9847}
          {arriveTangent=-0.0405, interpMode=RCIM_Cubic, leaveTangent=-0.0405, tangentMode=RCTM_User, time=0.3, value=0}
         ]}}})
     Z=0
     W=0
    )
   ) {isSetParameters=true, module=SetVariables_#11}
   /SetVariables(
    Particles.Color 2:LinearColor=&MultiplyLinearColor(
     Linear Color=$Particles.Initial.Color
     Scale Factor=&MakeVector4_FromLinearColor(LinearColor=&ColorFromCurve(ColorCurve=DataInterface() {propertyValues {
         AlphaCurve {keys [
           {arriveTangent=-0.1068, interpMode=RCIM_Cubic, leaveTangent=-0.1068, tangentMode=RCTM_User, time=0, value=1}
           {arriveTangent=-3.8787, interpMode=RCIM_Cubic, leaveTangent=-3.8787, tangentMode=RCTM_User, time=0.8779, value=0}
          ]}
         BlueCurve {keys [<at 0.2517 0.0595>,<at 0.6887 0.8958>]}
         GreenCurve {keys [<at 0.2517 0.0203>,<at 0.6887 0>]}
         RedCurve {keys [<at 0.2517 1>,<at 0.6887 0.2341>]}
        }}))
    )
    Particles.Dynamic0_2:Vector4f=&MakeVector4(
     X=&FloatFromCurve(FloatCurve=DataInterface() {propertyValues {Curve {keys [
          {interpMode=RCIM_Cubic, leaveTangent=0.923, tangentMode=RCTM_Break, time=0, value=0.525}
          {interpMode=RCIM_Cubic, tangentMode=RCTM_Break, time=0.6, value=0.8}
         ]}}})
     Y=0
     Z=0
     W=0
    )
   ) {isSetParameters=true, module=SetVariables_#12}
   /SetVariables(
    Particles.Color 3:LinearColor=&MultiplyLinearColor(
     Linear Color=$Particles.Initial.Color
     Scale Factor=&Vector4FromCurve(Vector4Curve=DataInterface() {propertyValues {
        AlphaCurve {keys [<at 0 1 RCIM_Cubic>, {arriveTangent=-2.5679, interpMode=RCIM_Cubic, leaveTangent=-2.5679, tangentMode=RCTM_User, time=0.8532, value=0}]}
        BlueCurve {keys [<at 0 0.0188>,<at 0.3176 0.0771>]}
        GreenCurve {keys [<at 0 0.0188>,<at 0.3176 0.006>]}
        RedCurve {keys [<at 0 0.0188>,<at 0.3176 0.0371>]}
       }})
    )
    Particles.Dynamic0_3:Vector4f=&MakeVector4(
     X=&FloatFromCurve(FloatCurve=DataInterface() {propertyValues {Curve {keys [
          {interpMode=RCIM_Cubic, leaveTangent=0.2062, tangentMode=RCTM_Break, time=0, value=0.9}
          {interpMode=RCIM_Cubic, tangentMode=RCTM_Break, time=0.2, value=1}
         ]}}})
     Y=0
     Z=0
     W=0
    )
   ) {isSetParameters=true, module=SetVariables_#13}
   /SetVariables(
    Particles.Color 4:LinearColor=&MultiplyLinearColor(
     Linear Color=$Particles.Initial.Color
     Scale Factor=&MakeVector4_FromLinearColor(LinearColor=&ColorFromCurve(ColorCurve=DataInterface() {propertyValues {
         AlphaCurve {keys [<at 0 1 RCIM_Cubic>, {arriveTangent=-4.1516, interpMode=RCIM_Cubic, leaveTangent=-4.1516, tangentMode=RCTM_User, time=0.8063, value=0}]}
         BlueCurve {keys [<at 0.2783 0.0262>,<at 0.6336 0.2521>]}
         GreenCurve {keys [<at 0.2783 0.0089>,<at 0.6336 0>]}
         RedCurve {keys [<at 0.2783 0.4396>,<at 0.6336 0.0659>]}
        }}))
    )
    Particles.Dynamic0_4:Vector4f=&MakeVector4(
     X=&FloatFromCurve(FloatCurve=DataInterface() {propertyValues {Curve {keys [
          {interpMode=RCIM_Cubic, leaveTangent=0.923, tangentMode=RCTM_Break, time=0, value=0.8}
          {interpMode=RCIM_Cubic, tangentMode=RCTM_Break, time=0.6, value=1}
         ]}}})
     Y=0
     Z=0
     W=0
    )
   ) {isSetParameters=true, module=SetVariables_#14}
   &ScaleSpriteSize(
    Uniform Curve Index=$Particles.PartitionParticles.Normalized Index In Partition
    Uniform Curve Sprite Scale=DataInterface() {propertyValues {Curve {keys [{tangentMode=RCTM_User, time=0, value=1}, <at 0.01 0>]}}}
   )
   &ScaleRibbonWidth(Ribbon Width Scale=&FloatFromCurve(
     FloatCurve=DataInterface() {propertyValues {Curve {keys [
         <at 0 0.5 RCIM_Cubic>
         {arriveTangentWeight=0.0373, interpMode=RCIM_Cubic, leaveTangentWeight=0.3354, tangentMode=RCTM_User, time=0.1, value=1}
         {arriveTangent=-2.2467, interpMode=RCIM_Cubic, leaveTangent=-2.2467, tangentMode=RCTM_User, time=1, value=0}
        ]}}}
     CurveIndex=$Particles.PartitionParticles.Normalized Index In Partition
     Scale Curve=1.5
    ))
   &ShapeLocation(
    MaskSpawnGroup=false
    Override Seed=false
    Shape Origin=$Output.FN_HermiteSpline.Position
    Sphere Radius=&FloatFromCurve(
     FloatCurve=DataInterface() {propertyValues {Curve {keys [<at 0 0 RCIM_Cubic>, {arriveTangent=3.426, interpMode=RCIM_Cubic, leaveTangent=3.426, tangentMode=RCTM_User, time=1, value=1}]}}}
     CurveIndex=$Particles.PartitionParticles.Normalized Index In Partition
     Scale Curve=&FloatFromCurve(
      FloatCurve=DataInterface() {propertyValues {Curve {keys [
          {interpMode=RCIM_Cubic, leaveTangentWeight=0.3333, tangentMode=RCTM_User, time=0, value=0}
          {arriveTangent=1.5942, arriveTangentWeight=0.3333, interpMode=RCIM_Cubic, leaveTangent=1.5942, tangentMode=RCTM_User, time=1, value=1}
         ]}}}
      Scale Curve=100
     )
    )
    UseOwnerScale=true
   )
   &ScaleSpriteSize(
    Uniform Curve Scale=&Waveform("Waveform [1]"=::NewEnumerator12, "[1] Frequency"=500, "[1] Amplitude Min/Max"=<0.5 1>)
    Uniform Curve Sprite Scale=DataInterface() {propertyValues {Curve {keys [
        {arriveTangent=6.2989, interpMode=RCIM_Cubic, leaveTangent=6.2989, tangentMode=RCTM_User, time=0, value=0}
        {interpMode=RCIM_Cubic, tangentMode=RCTM_User, time=0.45, value=1}
        {arriveTangent=-7.4736, tangentMode=RCTM_User, time=0.9, value=0}
       ]}}}
   ) {module=ScaleSpriteSize001}
   /SetVariables(
    Particles.Sprite Color 1:LinearColor=<rgba 0.0076 0.0023 0.0229 1>
    Particles.SpriteSize 1:Vector2f=&Multiply_Vector2D(A=$Particles.SpriteSize, B=&Vector2DFromFloat(Value=1.3))
   ) {isSetParameters=true, module=SetVariables_#15}
   /SetVariables(
    Particles.Sprite Color 2:LinearColor=<rgba 1.2 0.0243 0.0714 1>
    Particles.SpriteSize 2:Vector2f=&Multiply_Vector2D(A=$Particles.SpriteSize, B=&Vector2DFromFloat(Value=1))
   ) {isSetParameters=true, module=SetVariables_#16}
   /SetVariables(
    Particles.Sprite Color 3:LinearColor=<rgba 2.376 1.1885 3 1>
    Particles.SpriteSize 3:Vector2f=&Multiply_Vector2D(A=$Particles.SpriteSize, B=&Vector2DFromFloat(Value=0.424))
   ) {isSetParameters=true, module=SetVariables_#17}
   &FN_Light_Attributes(
    Color=&ColorFromCurve(ColorCurve=DataInterface() {propertyValues {
       AlphaCurve {keys [<at 0 1>,<at 0.7612 0>]}
       BlueCurve {keys [<at 0 0.0595>,<at 0.4522 0.8958>]}
       GreenCurve {keys [<at 0 0.0203>,<at 0.4522 0>]}
       RedCurve {keys [<at 0 1>,<at 0.4522 0.2341>]}
      }})
    Create Color Attribute=true
    Create Exponent Attribute=true
    Create Radius Attribute=true
    Enable Particle Light Enable=true
    Exponent=3
    Light Enabled:NiagaraBool=&SetBoolByFloatComparison(
     A=&FloatFromCurve(
      FloatCurve=DataInterface() {propertyValues {Curve {keys [<at 0 1>,<at 0.01 0>]}}}
      CurveIndex=$Particles.PartitionParticles.Normalized Index In Partition
     )
     B=0.5
    )
    Radius=27
    Scale Color=5
   )
  ]
  rendererClasses [
   &NiagaraRibbonRendererProperties
   &NiagaraSpriteRendererProperties
   &NiagaraLightRendererProperties
  ]
  renderers [
   {rendererClass=&NiagaraRibbonRendererProperties}
   {rendererClass=&NiagaraRibbonRendererProperties, rendererIndex=1}
   {rendererClass=&NiagaraRibbonRendererProperties, rendererIndex=2}
   {rendererClass=&NiagaraRibbonRendererProperties, rendererIndex=3}
   {rendererClass=&NiagaraSpriteRendererProperties, rendererIndex=4}
   {rendererClass=&NiagaraSpriteRendererProperties, rendererIndex=5}
   {rendererClass=&NiagaraSpriteRendererProperties, rendererIndex=6}
   {rendererClass=&NiagaraLightRendererProperties, rendererIndex=7}
  ]
 }
 LightningOut_EpicCin {
  emitterSpawnScript [
   /SetVariables(Emitter.Particles In Ribbon:NiagaraInt32=&MakeCustomIntFromBool(
     Bool=&IsPlatformSetActive(New Platform Set=DataInterface() {propertyValues {Platforms {qualityLevelMask=-31}}})
     True INT=3
     False INT=&MakeCustomIntFromBool(
      Bool=&IsPlatformSetActive(New Platform Set=DataInterface() {propertyValues {Platforms {qualityLevelMask=-30}}})
      True INT=4
      False INT=&MakeCustomIntFromBool(
       Bool=&IsPlatformSetActive(New Platform Set=DataInterface() {propertyValues {Platforms {qualityLevelMask=-28}}})
       True INT=5
       False INT=6
      )
     )
    )) {isSetParameters=true, module=SetVariables_#2}
   /SetVariables(Emitter.NumTris:NiagaraInt32=&self:GetNumTris(Static Mesh:NiagaraDataInterfaceStaticMesh=$System.Static Mesh, Subtract1?:NiagaraBool=true)) {isSetParameters=true, module=SetVariables_#3}
  ]
  emitterUpdateScript [
   &EmitterState(
    Life Cycle Mode=::NewEnumerator1
    Loop Behavior=::NewEnumerator2
    Loop Count=6
    Loop Duration=&RandomRangeFloat(Minimum=0.14, Maximum=0.16)
    Recalculate Duration Each Loop=true
    UseLoopDelay=false
   )
   &SpawnBurst_Instantaneous(Spawn Count=$Emitter.Particles In Ribbon, Use Spawn Probability=false, UseLoopCountLimit=false)
   /SetVariables(Emitter.TriToSample:NiagaraInt32=&RandomRangeInt(Minimum=0, Maximum=$Emitter.NumTris)) {isSetParameters=true, module=SetVariables_#4}
   /SetVariables(Emitter.Rand oFFset:NiagaraInt32=&MakeCustomIntFromBool(
     Bool=&RandomBool()
     True INT=&RandomRangeInt(Minimum=-8, Maximum=-4)
     False INT=&RandomRangeInt(Minimum=4, Maximum=8)
    )) {enabled=false, isSetParameters=true, module=SetVariables_#5}
   /SetVariables(Emitter.IsInside:NiagaraBool=&RandomBool(Recalculate Random Each Loop=true)) {enabled=false, isSetParameters=true, module=SetVariables_#7}
   /SetVariables(Emitter.Vector:Vector3f=&RandomRangeVector(Minimum=<0 0 0>, Maximum=<1 1 1>)) {isSetParameters=true}
  ]
  particleSpawnScript [
   &InitializeParticle(
    Color=<rgba 1 1 1 0.8>
    Lifetime=0.7
    Ribbon Width=25
    Ribbon Width Mode=::NewEnumerator1
    Sprite Size Mode=::NewEnumerator3
    Uniform Sprite Size=25
    UsePositionOffset=false
    Write Material Random=true
   )
   /SetVariables(Particles.GetTriToSample:NiagaraInt32=$Emitter.TriToSample) {isSetParameters=true, module=SetVariables_#8}
   /SetVariables(Particles.GetRandomOffset:NiagaraInt32=$Emitter.Rand oFFset) {isSetParameters=true, module=SetVariables_#9}
   /SetVariables(Particles.GetIsInside:NiagaraBool=$Emitter.IsInside) {isSetParameters=true, module=SetVariables_#10}
  ]
  particleUpdateScript [
   &ParticleState()
   &SpriteRotationRate()
   &self:ScratchModule(Time:NiagaraFloat=&Multiply_Float(A=$Engine.Time, B=720)) {module=ScratchModule}
   &PartitionParticles(
    Normalized Index In Partition Includes One=true
    Particles Per Partition:NiagaraInt32=$Emitter.Particles In Ribbon
    Partition Count Mode:ENiagara_InfiniteLoopDuration=::NewEnumerator1
    Partition Mode:ENiagara_PartitionMode=::NewEnumerator1
    Write Partition Index to Ribbon Link Order:ENiagara_AttributeSamplingApplyOutput=::NewEnumerator0
    Write Partition to Ribbon ID=::NewEnumerator0
   )
   &StaticMeshLocation(
    FixedBarycentric=false
    MaskSection=false
    MaskSpawnGroup=false
    Offset Position Along Sampled Normal=10
    OffsetAlongNormal=true
    SamplePosOffset=false
    Static Mesh:NiagaraDataInterfaceStaticMesh=$System.Static Mesh
    Triangle ID:NiagaraInt32=$Particles.GetTriToSample
    Triangle Sampling Mode:ENiagara_StaticTriangleMode=::NewEnumerator2
    Write To Particle Position=false
   )
   &self:RayCollision(
    IniPos:NiagaraPosition=&Lerp_Vector(
     A=&AddVectorToPosition(
      Position=$Particles.StaticMeshLocation.SampledPosition
      Vector=&Multiply_VectorByFloat(Vector=$Particles.StaticMeshLocation.SampledNormal, Float=4)
     )
     B=&AddVectorToPosition(
      Position=$Particles.StaticMeshLocation.SampledPosition
      Vector=&Multiply_VectorByFloat(Vector=$Particles.StaticMeshLocation.SampledNormal, Float=150)
     )
     Alpha=&FloatFromCurve(
      FloatCurve=DataInterface() {propertyValues {Curve {keys [{tangentMode=RCTM_User, time=0.01, value=0}, {tangentMode=RCTM_User, time=1, value=1}]}}}
      CurveIndex=&self:pixelate_Float(float:NiagaraFloat=$Particles.NormalizedAge, FPS:NiagaraFloat=6)
     )
    )
    SkipTrace:NiagaraBool=&IsPlatformSetActive(New Platform Set=DataInterface() {propertyValues {Platforms {qualityLevelMask=-25}}})
    TraceEndWorldOffset:NiagaraPosition=<0 0 -250>
    TraceStartWorldOffset:NiagaraPosition=<0 0 75>
   ) {module=RayCollision}
   &FN_HermiteSpline(
    Direction End:Vector3f=&Multiply_Vector(
     A=$Particles.CollisionNormal
     B=&VectorFromFloat(Value=&FloatFromCurve(
       FloatCurve=DataInterface() {propertyValues {Curve {keys [<at 0 0.5>,<at 1 1>]}}}
       CurveIndex=&self:pixelate_Float(float:NiagaraFloat=$Particles.NormalizedAge, FPS:NiagaraFloat=6)
       Scale Curve=150
      ))
    )
    Direction Start:Vector3f=&Multiply_Vector(A=$Particles.StaticMeshLocation.SampledNormal, B=&VectorFromFloat(Value=75))
    Position Blend:NiagaraFloat=$Particles.PartitionParticles.Normalized Index In Partition
    Position End:NiagaraPosition=$Particles.HitPostion
    Position Start:Vector3f=&AddVectorToPosition(
     Position=$Particles.StaticMeshLocation.SampledPosition
     Vector=&Multiply_VectorByFloat(Vector=$Particles.StaticMeshLocation.SampledNormal, Float=4)
    )
    Write Particles Position=true
   )
   /SetVariables(
    Particles.Color 1:LinearColor=&MultiplyLinearColor(
     Linear Color=$Particles.Initial.Color
     Scale Factor=&MakeVector4_FromLinearColor(LinearColor=&ColorFromCurve(ColorCurve=DataInterface() {propertyValues {
         AlphaCurve {keys [<at 0.5083 1>,<at 0.998 0>]}
         BlueCurve {keys [<at 0.5083 2>,<at 0.9881 1>]}
         GreenCurve {keys [<at 0.5083 2>,<at 0.9881 1>]}
         RedCurve {keys [<at 0.5083 2>,<at 0.9881 1>]}
        }}))
    )
    Particles.Dynamic0_1:Vector4f=&MakeVector4(
     X=&FloatFromCurve(FloatCurve=DataInterface() {propertyValues {Curve {keys [
          {interpMode=RCIM_Cubic, leaveTangent=0.4795, tangentMode=RCTM_Break, time=0, value=0.85}
          {interpMode=RCIM_Cubic, tangentMode=RCTM_Break, time=0.6, value=0.95}
         ]}}})
     Y=&FloatFromCurve(FloatCurve=DataInterface() {propertyValues {Curve {keys [
          {arriveTangent=9.0725, interpMode=RCIM_Cubic, leaveTangent=9.0725, tangentMode=RCTM_User, time=0, value=0}
          {arriveTangent=-2.5249, interpMode=RCIM_Cubic, leaveTangent=-19.203, tangentMode=RCTM_Break, time=0.1408, value=0.9847}
          {arriveTangent=-0.0405, interpMode=RCIM_Cubic, leaveTangent=-0.0405, tangentMode=RCTM_User, time=0.3, value=0}
         ]}}})
     Z=0
     W=0
    )
   ) {isSetParameters=true, module=SetVariables_#11}
   /SetVariables(
    Particles.Color 2:LinearColor=&MultiplyLinearColor(
     Linear Color=$Particles.Initial.Color
     Scale Factor=&MakeVector4_FromLinearColor(LinearColor=&ColorFromCurve(ColorCurve=DataInterface() {propertyValues {
         AlphaCurve {keys [<at 0.8328 1>,<at 1 0>]}
         BlueCurve {keys [<at 0.7364 0.0595>,<at 0.8603 0.0019>,<at 1 0.8958>]}
         GreenCurve {keys [<at 0.7364 0.0203>,<at 0.8603 0>,<at 1 0>]}
         RedCurve {keys [<at 0.7364 1>,<at 0.8603 0.1927>,<at 1 0.2341>]}
        }}))
    )
    Particles.Dynamic0_2:Vector4f=&MakeVector4(
     X=&FloatFromCurve(FloatCurve=DataInterface() {propertyValues {Curve {keys [
          {interpMode=RCIM_Cubic, leaveTangent=0.923, tangentMode=RCTM_Break, time=0, value=0.525}
          {interpMode=RCIM_Cubic, tangentMode=RCTM_Break, time=0.6, value=0.8}
         ]}}})
     Y=0
     Z=0
     W=0
    )
   ) {isSetParameters=true, module=SetVariables_#12}
   /SetVariables(
    Particles.Color 3:LinearColor=&MultiplyLinearColor(
     Linear Color=$Particles.Initial.Color
     Scale Factor=&Vector4FromCurve(Vector4Curve=DataInterface() {propertyValues {
        AlphaCurve {keys [<at 0.5361 1>,<at 0.9943 0>]}
        BlueCurve {keys [<at 0.7896 0.0188>,<at 0.9868 0.0771>]}
        GreenCurve {keys [<at 0.7896 0.0188>,<at 0.9868 0.006>]}
        RedCurve {keys [<at 0.7896 0.0188>,<at 0.9868 0.0371>]}
       }})
    )
    Particles.Dynamic0_3:Vector4f=&MakeVector4(
     X=&FloatFromCurve(FloatCurve=DataInterface() {propertyValues {Curve {keys [
          {interpMode=RCIM_Cubic, leaveTangent=0.2062, tangentMode=RCTM_Break, time=0, value=0.9}
          {interpMode=RCIM_Cubic, tangentMode=RCTM_Break, time=0.2, value=1}
         ]}}})
     Y=0
     Z=0
     W=0
    )
   ) {isSetParameters=true, module=SetVariables_#13}
   /SetVariables(
    Particles.Color 4:LinearColor=&MultiplyLinearColor(
     Linear Color=$Particles.Initial.Color
     Scale Factor=&MakeVector4_FromLinearColor(LinearColor=&ColorFromCurve(ColorCurve=DataInterface() {propertyValues {
         AlphaCurve {keys [<at 0.8269 1.5>,<at 0.996 0>]}
         BlueCurve {keys [<at 0.7246 0.0262>,<at 0.8623 0>,<at 0.994 0.2521>]}
         GreenCurve {keys [<at 0.7246 0.0089>,<at 0.8623 0>,<at 0.994 0>]}
         RedCurve {keys [<at 0.7246 0.4396>,<at 0.8623 0.1875>,<at 0.994 0.0659>]}
        }}))
    )
    Particles.Dynamic0_4:Vector4f=&MakeVector4(
     X=&FloatFromCurve(FloatCurve=DataInterface() {propertyValues {Curve {keys [
          {interpMode=RCIM_Cubic, leaveTangent=0.923, tangentMode=RCTM_Break, time=0, value=0.8}
          {interpMode=RCIM_Cubic, tangentMode=RCTM_Break, time=0.6, value=1}
         ]}}})
     Y=0
     Z=0
     W=0
    )
   ) {isSetParameters=true, module=SetVariables_#14}
   &ScaleSpriteSize(
    Uniform Curve Index=$Particles.PartitionParticles.Normalized Index In Partition
    Uniform Curve Sprite Scale=DataInterface() {propertyValues {Curve {keys [{tangentMode=RCTM_User, time=0, value=1}, <at 0.01 0>]}}}
   ) {enabled=false}
   &ScaleSpriteSize(
    Scale Sprite Size Mode=::NewEnumerator0
    Uniform Scale Factor=&MakeCustomFloatFromBool(
     Bool=$Particles.CollisionHappend
     True Float=&FloatFromCurve(
      FloatCurve=DataInterface() {propertyValues {Curve {keys [{tangentMode=RCTM_User, time=0, value=1}, <at 0.01 0>, <at 0.99 0>, <at 1 1>]}}}
      CurveIndex=$Particles.PartitionParticles.Normalized Index In Partition
     )
     False Float=&FloatFromCurve(
      FloatCurve=DataInterface() {propertyValues {Curve {keys [{tangentMode=RCTM_User, time=0, value=1}, <at 0.01 0>]}}}
      CurveIndex=$Particles.PartitionParticles.Normalized Index In Partition
     )
    )
   ) {module=ScaleSpriteSize002}
   &ScaleRibbonWidth(Ribbon Width Scale=&MakeCustomFloatFromBool(
     Bool=$Particles.CollisionHappend
     True Float=&FloatFromCurve(
      FloatCurve=DataInterface() {propertyValues {Curve {keys [
          <at 0 0 RCIM_Cubic>
          {
           arriveTangentWeight=0.0333
           interpMode=RCIM_Cubic
           leaveTangentWeight=0.4263
           tangentMode=RCTM_Break
           tangentWeightMode=RCTWM_WeightedBoth
           time=0.1
           value=1
          }
          {arriveTangent=-2.2894, interpMode=RCIM_Cubic, leaveTangent=-2.2894, tangentMode=RCTM_User, time=1, value=0.5}
         ]}}}
      CurveIndex=$Particles.PartitionParticles.Normalized Index In Partition
      Scale Curve=1.5
     )
     False Float=&FloatFromCurve(
      FloatCurve=DataInterface() {propertyValues {Curve {keys [
          <at 0 0 RCIM_Cubic>
          {
           arriveTangentWeight=0.0333
           interpMode=RCIM_Cubic
           leaveTangentWeight=0.73
           tangentMode=RCTM_Break
           tangentWeightMode=RCTWM_WeightedBoth
           time=0.1
           value=1
          }
          {arriveTangent=-2.9383, interpMode=RCIM_Cubic, leaveTangent=-2.9383, tangentMode=RCTM_User, time=1, value=0}
         ]}}}
      CurveIndex=$Particles.PartitionParticles.Normalized Index In Partition
      Scale Curve=1.5
     )
    ))
   &ShapeLocation(
    MaskSpawnGroup=false
    Override Seed=false
    Shape Origin=$Output.FN_HermiteSpline.Position
    Sphere Radius=&FloatFromCurve(
     FloatCurve=DataInterface() {propertyValues {Curve {keys [<at 0 0 RCIM_Cubic>, {arriveTangent=3.426, interpMode=RCIM_Cubic, leaveTangent=3.426, tangentMode=RCTM_User, time=1, value=1}]}}}
     CurveIndex=$Particles.PartitionParticles.Normalized Index In Partition
     Scale Curve=&FloatFromCurve(
      FloatCurve=DataInterface() {propertyValues {Curve {keys [
          {interpMode=RCIM_Cubic, leaveTangentWeight=0.3333, tangentMode=RCTM_User, time=0, value=0}
          {arriveTangent=1.5942, arriveTangentWeight=0.3333, interpMode=RCIM_Cubic, leaveTangent=1.5942, tangentMode=RCTM_User, time=1, value=1}
         ]}}}
      CurveIndex=&self:pixelate_Float(float:NiagaraFloat=$Particles.NormalizedAge, FPS:NiagaraFloat=6)
      Scale Curve=5
     )
    )
    UseOwnerScale=true
   ) {enabled=false}
   &ScaleSpriteSize(
    Uniform Curve Scale=&Waveform("Waveform [1]"=::NewEnumerator12, "[1] Frequency"=500, "[1] Amplitude Min/Max"=<0.5 1>)
    Uniform Curve Sprite Scale=DataInterface() {propertyValues {Curve {keys [
        {arriveTangent=6.2989, interpMode=RCIM_Cubic, leaveTangent=6.2989, tangentMode=RCTM_User, time=0, value=0}
        {interpMode=RCIM_Cubic, tangentMode=RCTM_User, time=0.45, value=1}
        {arriveTangent=-7.4736, tangentMode=RCTM_User, time=0.9, value=0}
       ]}}}
   ) {module=ScaleSpriteSize001}
   /SetVariables(
    Particles.Sprite Color 1:LinearColor=<rgba 0.0076 0.0023 0.0229 1>
    Particles.SpriteSize 1:Vector2f=&Multiply_Vector2D(A=$Particles.SpriteSize, B=&Vector2DFromFloat(Value=1.3))
   ) {isSetParameters=true, module=SetVariables_#15}
   /SetVariables(
    Particles.Sprite Color 2:LinearColor=&MultiplyLinearColor(
     Linear Color=<rgba 0.5 0.0101 0.0101 1>
     Scale Factor=&MakeFloat4FromVectorAndFloat(
      "Vector (RGB)"=<1 1 1>
      "Float (Alpha)"=&FloatFromCurve(FloatCurve=DataInterface() {propertyValues {Curve {keys [<at 0.8 1>,<at 1 0>]}}})
     )
    )
    Particles.SpriteSize 2:Vector2f=&Multiply_Vector2D(A=$Particles.SpriteSize, B=&Vector2DFromFloat(Value=1))
   ) {isSetParameters=true, module=SetVariables_#16}
   /SetVariables(
    Particles.Sprite Color 3:LinearColor=<rgba 2.376 1.1885 3 1>
    Particles.SpriteSize 3:Vector2f=&Multiply_Vector2D(A=$Particles.SpriteSize, B=&Vector2DFromFloat(Value=0.424))
   ) {isSetParameters=true, module=SetVariables_#17}
   &FN_Light_Attributes(
    Color=&ColorFromCurve(ColorCurve=DataInterface() {propertyValues {
       AlphaCurve {keys [<at 0 1>,<at 0.9936 0>]}
       BlueCurve {keys [<at 0.774 0.0595>,<at 0.9664 0.8958>]}
       GreenCurve {keys [<at 0.774 0.0203>,<at 0.9664 0>]}
       RedCurve {keys [<at 0.774 1>,<at 0.9664 0.2341>]}
      }})
    Create Color Attribute=true
    Create Exponent Attribute=true
    Create Position Attribute=true
    Create Radius Attribute=true
    Enable Particle Light Enable=true
    Exponent=3
    Light Enabled:NiagaraBool=&SetBoolByFloatComparison(
     A=&MakeCustomFloatFromBool(
      Bool=$Particles.CollisionHappend
      True Float=&FloatFromCurve(
       FloatCurve=DataInterface() {propertyValues {Curve {keys [<at 0 1>,<at 0.01 0>,<at 0.99 0>,<at 1 1>]}}}
       CurveIndex=$Particles.PartitionParticles.Normalized Index In Partition
      )
      False Float=&FloatFromCurve(
       FloatCurve=DataInterface() {propertyValues {Curve {keys [<at 0 1>,<at 0.01 0>]}}}
       CurveIndex=$Particles.PartitionParticles.Normalized Index In Partition
      )
     )
     B=0.5
    )
    Position=&AddVectorToPosition(
     Position=$Particles.Position
     Vector=&Multiply_Vector(A=$Particles.CollisionNormal, B=&VectorFromFloat(Value=3))
    )
    Radius=27
    Scale Color=&FloatFromCurve(
     FloatCurve=DataInterface() {propertyValues {Curve {keys [<at 0 1>,<at 0.01 0>,<at 0.99 0>,<at 1 3>]}}}
     CurveIndex=$Particles.PartitionParticles.Normalized Index In Partition
     Scale Curve=5
    )
   )
   &CameraOffset(Camera Offset Amount=&FloatFromCurve(
     FloatCurve=DataInterface() {propertyValues {Curve {keys [<at 0.99 0>,<at 1 1>]}}}
     CurveIndex=$Particles.PartitionParticles.Normalized Index In Partition
     Scale Curve=25
    ))
   &self:ScratchModule_01(
    Offset:Vector3f=&RandomVector(Vector Scale=&Multiply_Float(
      A=&FloatFromCurve(
       FloatCurve=DataInterface() {propertyValues {Curve {keys [
           {arriveTangent=5.6674, interpMode=RCIM_Cubic, leaveTangent=5.6674, tangentMode=RCTM_User, time=0.05, value=0}
           {interpMode=RCIM_Cubic, tangentMode=RCTM_User, time=0.5, value=1}
           {arriveTangent=-4.7955, interpMode=RCIM_Cubic, leaveTangent=-4.7955, tangentMode=RCTM_User, time=0.95, value=0}
          ]}}}
       CurveIndex=$Particles.PartitionParticles.Normalized Index In Partition
      )
      B=&RandomRangeFloat(Minimum=0, Maximum=25, Evaluation Type=::NewEnumerator1)
     ))
    Timer Amount:NiagaraFloat=&Divide_Float(A=1, B=6)
   ) {module=ScratchModule_01}
  ]
  rendererClasses [
   &NiagaraRibbonRendererProperties
   &NiagaraSpriteRendererProperties
   &NiagaraLightRendererProperties
  ]
  renderers [
   {rendererClass=&NiagaraRibbonRendererProperties}
   {rendererClass=&NiagaraRibbonRendererProperties, rendererIndex=1}
   {rendererClass=&NiagaraRibbonRendererProperties, rendererIndex=2}
   {rendererClass=&NiagaraRibbonRendererProperties, rendererIndex=3}
   {rendererClass=&NiagaraSpriteRendererProperties, rendererIndex=4}
   {rendererClass=&NiagaraSpriteRendererProperties, rendererIndex=5}
   {rendererClass=&NiagaraSpriteRendererProperties, rendererIndex=6}
   {rendererClass=&NiagaraLightRendererProperties, rendererIndex=7}
  ]
 }
 Poof {
  emitterSpawnScript [/SetVariables(Emitter.NumParts:NiagaraInt32=&MakeCustomIntFromBool(
     Bool=&IsPlatformSetActive(New Platform Set=DataInterface() {propertyValues {Platforms {qualityLevelMask=-31}}})
     True INT=8
     False INT=&MakeCustomIntFromBool(
      Bool=&IsPlatformSetActive(New Platform Set=DataInterface() {propertyValues {Platforms {qualityLevelMask=-30}}})
      True INT=12
      False INT=&MakeCustomIntFromBool(
       Bool=&IsPlatformSetActive(New Platform Set=DataInterface() {propertyValues {Platforms {qualityLevelMask=-28}}})
       True INT=14
       False INT=16
      )
     )
    )) {isSetParameters=true, module=SetVariables_#18}]
  emitterUpdateScript [
   &EmitterState(Life Cycle Mode=::NewEnumerator1, Loop Behavior=::NewEnumerator1, Loop Duration=1, UseLoopDelay=true)
   &SpawnBurst_Instantaneous(Spawn Count=$Emitter.NumParts, Use Spawn Probability=false, UseLoopCountLimit=false)
   /SetVariables(Emitter.FELight:LinearColor=&MakeCustomLinearColorFromBool(Bool=$User.IsFrontEnd, True Color=<rgba 0.218 -0.3992 1.0836 1>, False Color=<rgba 0 0 0 0>)) {isSetParameters=true, module=SetVariables_#19}
   /SetVariables(Emitter.TOD:Vector3f=&MakeCustomVectorFromBool(Bool=$User.IsFrontEnd, True Vector=<0 1 1>, False Vector=&MakeVector(X=0, Y=1, Z=1))) {isSetParameters=true, module=SetVariables_#20}
   /SetVariables(Emitter.LitRim:LinearColor=&MakeCustomLinearColorFromBool(Bool=$User.IsFrontEnd, True Color=<rgba 0.5 0.3615 0.5 1>, False Color=<rgba 0.5 0.3615 0.5 1>)) {isSetParameters=true, module=SetVariables_#21}
   /SetVariables(Emitter.Lit:LinearColor=&MakeCustomLinearColorFromBool(Bool=$User.IsFrontEnd, True Color=<rgba 0.5 0.3615 0.5 1>, False Color=<rgba 0.5 0.3615 0.5 1>)) {isSetParameters=true, module=SetVariables_#22}
   /SetVariables(Emitter.Shadow:LinearColor=&MakeCustomLinearColorFromBool(Bool=$User.IsFrontEnd, True Color=<rgba 0.4564 0.2961 0.5149 1>, False Color=<rgba 0.4564 0.2961 0.5149 1>)) {isSetParameters=true, module=SetVariables_#23}
   /SetVariables(Emitter.GI:LinearColor=&MakeCustomLinearColorFromBool(Bool=$User.IsFrontEnd, True Color=<rgba 0.2371 0.1312 0.35 1>, False Color=<rgba 0.2371 0.1312 0.35 1>)) {isSetParameters=true, module=SetVariables_#24}
   /SetVariables(Emitter.SoftShadowStr:NiagaraFloat=&MakeCustomFloatFromBool(Bool=$User.IsFrontEnd, True Float=0.6, False Float=0.5)) {isSetParameters=true, module=SetVariables_#25}
   /Game/Effects/Niagara/MasterMaterials/ToonShader/Template/NM_OutlineParameters()
  ]
  particleSpawnScript [
   /SetVariables() {isSetParameters=true, module=SetVariables_#26}
   &InitializeParticle(
    Color Mode=::NewEnumerator0
    Lifetime=&Lerp_Float(
     A=0.4
     B=1
     Alpha=&FloatFromCurve(
      FloatCurve=DataInterface() {propertyValues {Curve {keys [
          {interpMode=RCIM_Cubic, leaveTangentWeight=0.4723, tangentMode=RCTM_Break, tangentWeightMode=RCTWM_WeightedBoth, time=0, value=0}
          {
           arriveTangent=0.5709
           arriveTangentWeight=0.8122
           interpMode=RCIM_Cubic
           leaveTangent=0.5709
           leaveTangentWeight=0.8122
           tangentMode=RCTM_User
           tangentWeightMode=RCTWM_WeightedBoth
           time=1
           value=1
          }
         ]}}}
      CurveIndex=&RandomRangeFloat(Minimum=0, Maximum=1)
     )
    )
    Mesh Scale=&Multiply_VectorByFloat(
     Vector=&RandomRangeVector(Minimum=<0.7 0.7 4>, Maximum=<1.35 3 6>)
     Float=&RandomRangeFloat(Minimum=0.22, Maximum=0.05)
    )
    Mesh Scale Mode=::NewEnumerator4
    Sprite Rotation Mode=::NewEnumerator2
    Sprite Size Mode=::NewEnumerator3
    Uniform Sprite Size=&MakeCustomFloatFromBool(
     Bool=&SetBoolByIntComparison(
      A=&ModuloInt(A=&ReturnParticleIDAttributes(), B=2)
      B=0
      Comparison Type=::NewEnumerator0
     )
     True Float=200
     False Float=0
    )
    UsePositionOffset=true
    Write Material Random=true
   )
   &PartitionParticles(
    Normalized Index In Partition Includes One=true
    Number Of Partitions=1
    Particles Per Partition:NiagaraInt32=$Emitter.NumParts
    Partition Mode:ENiagara_PartitionMode=::NewEnumerator1
   )
   &ShapeLocation(
    Cylinder Distribution=::NewEnumerator2
    Cylinder Height=0.01
    Cylinder Radius=40
    MaskSpawnGroup=false
    Shape Primitive=::NewEnumerator1
    U Position=&Lerp_Float(A=0, B=1, Alpha=$Particles.PartitionParticles.Normalized Index In Partition)
    UseOwnerScale=true
    V Position=1
   )
   &InitialMeshOrientation(
    Mesh Orientation Mode=::NewEnumerator2
    Orientation Coordinate Space=::World
    Orientation Vector=/Niagara/DynamicInputs/Vectors/CrossProduct(
     Vector A:Vector3f=&NormalizeVector(
      Vector To Normalize=&Subtract_Position(A=$Particles.Position, B=&SimulationPosition())
      Safe Normalize=true
      Safe Fallback Threshold=0
      Safe Fallback Vector=<0 0 0.001>
     )
     Vector B:Vector3f=<0 0 1>
     Normalize Incoming Vectors:NiagaraBool=true
    )
    Rotation Coordinate Space=::NewEnumerator1
    Use Rotation Vector=true
   )
   &AddVelocity(
    Module.Use Falloff Distance=false
    Origin Offset=<0 0 -20>
    Override Seed=false
    Velocity Mode=::NewEnumerator1
    Velocity Speed=&RandomRangeFloat(Minimum=500, Maximum=350)
   )
  ]
  particleUpdateScript [
   &ParticleState()
   &ScaleMeshSize() {module=ScaleMeshSize002}
   &ScaleMeshSize(Scale Factor=&MakeVector(
     X=1
     Y=1
     Z=&FloatFromCurve(FloatCurve=DataInterface() {propertyValues {Curve {keys [{arriveTangent=-7.7792, interpMode=RCIM_Cubic, leaveTangent=-7.7792, tangentMode=RCTM_User, time=0, value=2.5}, <at 0.5 1>]}}})
    )) {module=ScaleMeshSize001}
   &ScaleMeshSize(Scale Factor=&VectorFromFloat(Value=&Lerp_Float(
      A=&FloatFromCurve(FloatCurve=DataInterface() {propertyValues {Curve {keys [
           {arriveTangent=6.2494, interpMode=RCIM_Cubic, leaveTangent=6.2494, tangentMode=RCTM_User, time=0, value=0}
           {interpMode=RCIM_Cubic, tangentMode=RCTM_User, time=0.5, value=1}
           <at 1 0 RCIM_Cubic>
          ]}}})
      B=&FloatFromCurve(FloatCurve=DataInterface() {propertyValues {Curve {keys [
           {arriveTangent=3.1805, interpMode=RCIM_Cubic, leaveTangent=3.1805, tangentMode=RCTM_User, time=0, value=0.1}
           {arriveTangent=1.625, interpMode=RCIM_Cubic, leaveTangent=1.625, time=0.15, value=1.1}
           {arriveTangentWeight=0.2065, interpMode=RCIM_Cubic, leaveTangentWeight=0.0688, tangentMode=RCTM_User, time=0.8, value=1.4}
           {arriveTangent=-13.3162, interpMode=RCIM_Cubic, leaveTangent=-13.3162, tangentMode=RCTM_User, time=1, value=0}
          ]}}})
      Alpha=&Lerp_Float(
       A=0
       B=1
       Alpha=&self:inverseLerp(A:NiagaraFloat=0.4, B:NiagaraFloat=1.25, Clamp:NiagaraBool=true, V:NiagaraFloat=$Particles.Lifetime)
      )
     )))
   &UpdateMeshOrientation(
    Rotation Coordinate Space=::NewEnumerator1
    Rotation Rate=&Multiply_Float(A=0.05, B=&RandomRangeFloat(Minimum=-1, Maximum=1))
    Rotation Vector=<0 0 1>
   )
   &UpdateMeshOrientation(
    Rotation Rate=&Multiply_Float(A=0.15, B=&RandomRangeFloat(Minimum=-1, Maximum=1))
    Rotation Vector=&RandomVector(Vector Scale=1)
   ) {module=UpdateMeshOrientation001}
   &DynamicMaterialParameters(
    Index 0 Param 1=&MakeFloatFromVector(Vector=$Particles.Scale, Channel=::NewEnumerator0)
    Index 0 Param 2=&Add_Float(
     A=&FloatFromCurve(
      FloatCurve=DataInterface() {propertyValues {Curve {keys [
          {arriveTangent=2.4846, interpMode=RCIM_Cubic, leaveTangent=2.4846, tangentMode=RCTM_User, time=0, value=0}
          {arriveTangent=0.1519, tangentMode=RCTM_User, time=1, value=1}
         ]}}}
      Scale Curve=-10
     )
     B=&RandomRangeFloat(Minimum=0, Maximum=100)
    )
    Index 0 Param 3=0.7
    Module.Param0WriteEnabled=true
    Module.Param0WriteEnabled3=true
    Module.Param1WriteEnabled=true
    Module.Param1WriteEnabled3=true
    Module.Param2WriteEnabled=true
    Module.Param2WriteEnabled3=true
    Module.Param3WriteEnabled=true
    Module.Param3WriteEnabled3=true
    Write Parameter Index 3=true
   ) {module=DynamicMaterialParameters001}
   &DynamicMaterialParameters(
    Index 0 Param 1=&MakeFloatFromVector(Vector=$Particles.Scale, Channel=::NewEnumerator0)
    Index 0 Param 2=&Add_Float(
     A=&Lerp_Float(
      A=&FloatFromCurve(
       FloatCurve=DataInterface() {propertyValues {Curve {keys [{tangentMode=RCTM_User, time=0, value=0}, {tangentMode=RCTM_User, time=1, value=1}]}}}
       Scale Curve=0.5
      )
      B=&FloatFromCurve(
       FloatCurve=DataInterface() {propertyValues {Curve {keys [
           {arriveTangent=2.4846, interpMode=RCIM_Cubic, leaveTangent=2.4846, tangentMode=RCTM_User, time=0, value=0}
           {arriveTangent=0.1519, tangentMode=RCTM_User, time=1, value=1}
          ]}}}
       Scale Curve=-10
      )
      Alpha=&Lerp_Float(
       A=0
       B=1
       Alpha=&self:inverseLerp(A:NiagaraFloat=0.4, B:NiagaraFloat=1.25, Clamp:NiagaraBool=true, V:NiagaraFloat=$Particles.Lifetime)
      )
     )
     B=&Add_Float(
      A=&Multiply_Float(A=$Engine.Time, B=0.1)
      B=&RandomRangeFloat(Minimum=0, Maximum=100)
     )
    )
    Index 0 Param 3=0.7
    Index 3 Param 1=&FN_FrontEndFloatModifier(
     Gameplay=&FNTimeOfDayScalar(Night Scale=-0.0704, Morning Scale=-0.012, Day Scale=0, Evening Scale=0)
     Front End=-0.024
    )
    Index 3 Param 2=&FNTimeOfDayScalar(Night Scale=-0.228, Morning Scale=0.2, Day Scale=0, Evening Scale=0)
    Index 3 Param 3=&FN_FrontEndFloatModifier(Gameplay=&FNTimeOfDayScalar(Night Scale=0.588, Morning Scale=1, Evening Scale=1), Front End=1.3)
    Module.Param0WriteEnabled=true
    Module.Param0WriteEnabled1=true
    Module.Param0WriteEnabled3=true
    Module.Param1WriteEnabled=true
    Module.Param1WriteEnabled1=true
    Module.Param1WriteEnabled3=true
    Module.Param2WriteEnabled=true
    Module.Param2WriteEnabled1=true
    Module.Param2WriteEnabled3=true
    Module.Param3WriteEnabled=true
    Module.Param3WriteEnabled1=true
    Module.Param3WriteEnabled3=true
    Write Parameter Index 1=true
    Write Parameter Index 3=true
   )
   &Drag(Drag=4, Use Linear Drag=true, Use Rotational Drag=false)
   &GravityForce(Gravity=&Multiply_VectorByFloat(Vector=<0 0 -980>, Float=0.1))
   &SolveForcesAndVelocity(Clamp Velocity=false, Limit Acceleration=false)
   &ScaleColor(
    Scale Alpha=&FloatFromCurve(FloatCurve=DataInterface() {propertyValues {Curve {keys [<at 0.75 1>,<at 1 0>]}}})
    ScaleA=true
    ScaleRGB=true
   ) {enabled=false}
   &ScaleColor(
    Scale Alpha=&FloatFromCurve(FloatCurve=DataInterface() {propertyValues {Curve {keys [{arriveTangent=-3.9434, interpMode=RCIM_Cubic, leaveTangent=-3.9434, tangentMode=RCTM_User, time=0.2, value=1}, <at 1 0 RCIM_Cubic>]}}})
    ScaleA=true
    ScaleRGB=true
   ) {module=ScaleColor001}
   &ScaleSpriteSize(
    Uniform Curve Scale=1.1
    Uniform Curve Sprite Scale=DataInterface() {propertyValues {Curve {keys [
        {arriveTangent=3.1805, interpMode=RCIM_Cubic, leaveTangent=3.1805, leaveTangentWeight=0.3333, tangentMode=RCTM_User, time=0, value=0.1}
        {
         arriveTangent=1.625
         arriveTangentWeight=0.05
         interpMode=RCIM_Cubic
         leaveTangent=1.625
         leaveTangentWeight=0.2833
         tangentMode=RCTM_User
         time=0.15
         value=0.65
        }
        {arriveTangentWeight=0.2167, interpMode=RCIM_Cubic, leaveTangentWeight=0.0667, tangentMode=RCTM_User, time=0.8, value=1}
       ]}}}
   )
   /SetVariables(Particles.Sprite color:LinearColor=&MultiplyLinearColor(
     Linear Color=&ColorFromCurve(ColorCurve=DataInterface() {propertyValues {
        AlphaCurve {keys [<at 0 0>, <at 0.098 1>, {arriveTangent=-1.2798, interpMode=RCIM_Cubic, leaveTangent=-1.2798, time=0.4608, value=1}, <at 0.8793 0 RCIM_Cubic>]}
        BlueCurve {keys [<at 0.122 0.5 RCIM_Cubic>]}
        GreenCurve {keys [<at 0.122 0.3615 RCIM_Cubic>]}
        RedCurve {keys [<at 0.122 0.4077 RCIM_Cubic>]}
       }})
     Scale Factor=&MakeFloat4FromVectorAndFloat("Vector (RGB)"=<1 1 1>, "Float (Alpha)"=0.1)
    )) {isSetParameters=true, module=SetVariables_#27}
  ]
  rendererClasses [&NiagaraMeshRendererProperties, &NiagaraSpriteRendererProperties]
  renderers [
   {rendererClass=&NiagaraMeshRendererProperties, rendererIndex=0}
   {rendererClass=&NiagaraMeshRendererProperties, rendererIndex=1}
   {rendererClass=&NiagaraSpriteRendererProperties, rendererIndex=2}
   {rendererClass=&NiagaraSpriteRendererProperties, rendererIndex=3}
  ]
 }
 Poof001 {
  bEnabled=false
  emitterSpawnScript [/SetVariables(Emitter.NumParts:NiagaraInt32=&MakeCustomIntFromBool(
     Bool=&IsPlatformSetActive(New Platform Set=DataInterface() {propertyValues {Platforms {qualityLevelMask=-31}}})
     True INT=8
     False INT=&MakeCustomIntFromBool(
      Bool=&IsPlatformSetActive(New Platform Set=DataInterface() {propertyValues {Platforms {qualityLevelMask=-30}}})
      True INT=12
      False INT=&MakeCustomIntFromBool(
       Bool=&IsPlatformSetActive(New Platform Set=DataInterface() {propertyValues {Platforms {qualityLevelMask=-28}}})
       True INT=14
       False INT=16
      )
     )
    )) {isSetParameters=true, module=SetVariables_#18}]
  emitterUpdateScript [
   &EmitterState(Life Cycle Mode=::NewEnumerator1, Loop Behavior=::NewEnumerator1, Loop Duration=1, UseLoopDelay=true)
   &SpawnBurst_Instantaneous(Spawn Count=$Emitter.NumParts, Use Spawn Probability=false, UseLoopCountLimit=false)
   /SetVariables(Emitter.FELight:LinearColor=&MakeCustomLinearColorFromBool(
     Bool=$User.IsFrontEnd
     True Color=&MakeLinearColorFromVectorAndFloat(
      "Vector (RGB)"=&NormalizeVector(Vector To Normalize=<0 1 1>, Safe Normalize=true, Safe Fallback Threshold=0, Safe Fallback Vector=<0 0 0.001>)
      "Float (Alpha)"=1
     )
     False Color=<rgba 0 0 0 0>
    )) {isSetParameters=true, module=SetVariables_#19}
   /SetVariables(Emitter.TOD:Vector3f=&MakeCustomVectorFromBool(Bool=$User.IsFrontEnd, True Vector=<0 1 1>, False Vector=&MakeVector(X=0, Y=1, Z=1))) {isSetParameters=true, module=SetVariables_#20}
   /SetVariables(Emitter.LitRim:LinearColor=&MakeCustomLinearColorFromBool(Bool=$User.IsFrontEnd, True Color=<rgba 0.5 0.3615 0.5 1>, False Color=<rgba 0.5 0.3615 0.5 1>)) {isSetParameters=true, module=SetVariables_#21}
   /SetVariables(Emitter.Lit:LinearColor=&MakeCustomLinearColorFromBool(Bool=$User.IsFrontEnd, True Color=<rgba 0.5 0.3615 0.5 1>, False Color=<rgba 0.5 0.3615 0.5 1>)) {isSetParameters=true, module=SetVariables_#22}
   /SetVariables(Emitter.Shadow:LinearColor=&MakeCustomLinearColorFromBool(Bool=$User.IsFrontEnd, True Color=<rgba 0.4564 0.2961 0.5149 1>, False Color=<rgba 0.4564 0.2961 0.5149 1>)) {isSetParameters=true, module=SetVariables_#23}
   /SetVariables(Emitter.GI:LinearColor=&MakeCustomLinearColorFromBool(Bool=$User.IsFrontEnd, True Color=<rgba 0.2371 0.1312 0.35 1>, False Color=<rgba 0.2371 0.1312 0.35 1>)) {isSetParameters=true, module=SetVariables_#24}
   /SetVariables(Emitter.SoftShadowStr:NiagaraFloat=&MakeCustomFloatFromBool(Bool=$User.IsFrontEnd, True Float=0.6, False Float=0.5)) {isSetParameters=true, module=SetVariables_#25}
   &NM_OutlineParameters()
  ]
  particleSpawnScript [
   /SetVariables(Particles.MeshAxis:NiagaraInt32=&RandomRangeInt(Minimum=0, Maximum=1)) {isSetParameters=true, module=SetVariables_#26}
   &InitializeParticle(
    Color Mode=::NewEnumerator0
    Lifetime=&Lerp_Float(
     A=0.4
     B=1
     Alpha=&FloatFromCurve(
      FloatCurve=DataInterface() {propertyValues {Curve {keys [
          {interpMode=RCIM_Cubic, leaveTangentWeight=0.4723, tangentMode=RCTM_Break, tangentWeightMode=RCTWM_WeightedBoth, time=0, value=0}
          {
           arriveTangent=0.5709
           arriveTangentWeight=0.8122
           interpMode=RCIM_Cubic
           leaveTangent=0.5709
           leaveTangentWeight=0.8122
           tangentMode=RCTM_User
           tangentWeightMode=RCTWM_WeightedBoth
           time=1
           value=1
          }
         ]}}}
      CurveIndex=&RandomRangeFloat(Minimum=0, Maximum=1)
     )
    )
    Mesh Scale=&Multiply_VectorByFloat(
     Vector=&SelectVectorFromArray(Direct Array Index=0, Vector Selection Array=DataInterface() {propertyValues {FloatData [<3 1 1>, <1 1 3>]}})
     Float=&RandomRangeFloat(Minimum=0.035, Maximum=0.01)
    )
    Mesh Scale Mode=::NewEnumerator4
    Sprite Rotation Mode=::NewEnumerator2
    Sprite Size Mode=::NewEnumerator3
    Uniform Sprite Size=&MakeCustomFloatFromBool(
     Bool=&SetBoolByIntComparison(
      A=&ModuloInt(A=&ReturnParticleIDAttributes(), B=2)
      B=0
      Comparison Type=::NewEnumerator0
     )
     True Float=200
     False Float=0
    )
    UsePositionOffset=true
    Write Material Random=true
   )
   &PartitionParticles(
    Normalized Index In Partition Includes One=true
    Number Of Partitions=1
    Particles Per Partition:NiagaraInt32=$Emitter.NumParts
    Partition Mode:ENiagara_PartitionMode=::NewEnumerator1
   )
   &ShapeLocation(
    Cylinder Distribution=::NewEnumerator2
    Cylinder Height=0.01
    Cylinder Radius=40
    MaskSpawnGroup=false
    Shape Primitive=::NewEnumerator1
    U Position=&Lerp_Float(A=0, B=1, Alpha=$Particles.PartitionParticles.Normalized Index In Partition)
    UseOwnerScale=true
    V Position=1
   )
   &InitialMeshOrientation(
    Mesh Axis To Orient=&SelectVectorFromArray(Direct Array Index=$Particles.MeshAxis, Vector Selection Array=DataInterface() {propertyValues {FloatData [<1 0 0>, <0 1 0>]}})
    Mesh Orientation Mode=::NewEnumerator2
    Orientation Coordinate Space=::World
    Orientation Vector=&CrossProduct(
     Vector A:Vector3f=&NormalizeVector(
      Vector To Normalize=&Subtract_Position(A=$Particles.Position, B=&SimulationPosition())
      Safe Normalize=true
      Safe Fallback Threshold=0
      Safe Fallback Vector=<0 0 0.001>
     )
     Vector B:Vector3f=<0 0 1>
     Normalize Incoming Vectors:NiagaraBool=true
    )
    Rotation=&SelectVectorFromArray(Direct Array Index=$Particles.MeshAxis, Vector Selection Array=DataInterface() {propertyValues {FloatData [<0 0 0>, <0 0 0.25>]}})
    Rotation Coordinate Space=::NewEnumerator1
    Use Rotation Vector=true
   )
   &AddVelocity(
    Module.Use Falloff Distance=false
    Origin Offset=<0 0 -20>
    Override Seed=false
    Velocity Mode=::NewEnumerator1
    Velocity Speed=&RandomRangeFloat(Minimum=500, Maximum=350)
   )
  ]
  particleUpdateScript [
   &ParticleState()
   &ScaleMeshSize(Scale Factor=&VectorFromCurve(
     VectorCurve=DataInterface() {propertyValues {XCurve {keys [<at 0 1>,<at 0.25 1>]}, YCurve {keys [<at 0 0.35>,<at 0.25 1>]}, ZCurve {keys [<at 0 1>,<at 0.25 1>]}}}
     CurveIndex=$Emitter.Age
    )) {module=ScaleMeshSize002}
   &ScaleMeshSize(Scale Factor=&VectorFromFloat(Value=&Lerp_Float(
      A=&FloatFromCurve(FloatCurve=DataInterface() {propertyValues {Curve {keys [
           {arriveTangent=6.2494, interpMode=RCIM_Cubic, leaveTangent=6.2494, tangentMode=RCTM_User, time=0, value=0}
           {interpMode=RCIM_Cubic, tangentMode=RCTM_User, time=0.5, value=1}
           <at 1 0 RCIM_Cubic>
          ]}}})
      B=&FloatFromCurve(FloatCurve=DataInterface() {propertyValues {Curve {keys [
           {arriveTangent=3.1805, interpMode=RCIM_Cubic, leaveTangent=3.1805, tangentMode=RCTM_User, time=0, value=0.1}
           {arriveTangent=1.625, interpMode=RCIM_Cubic, leaveTangent=1.625, time=0.15, value=1.1}
           {arriveTangentWeight=0.2065, interpMode=RCIM_Cubic, leaveTangentWeight=0.0688, tangentMode=RCTM_User, time=0.8, value=1.4}
           {arriveTangent=-13.3162, interpMode=RCIM_Cubic, leaveTangent=-13.3162, tangentMode=RCTM_User, time=1, value=0}
          ]}}})
      Alpha=&Lerp_Float(
       A=0
       B=1
       Alpha=&self:inverseLerp(A:NiagaraFloat=0.4, B:NiagaraFloat=1.25, Clamp:NiagaraBool=true, V:NiagaraFloat=$Particles.Lifetime)
      )
     )))
   &ScaleMeshSize(Scale Factor=&self:ScaleSwizzle(
     Default:Vector3f=&VectorFromCurve(VectorCurve=DataInterface() {propertyValues {XCurve {keys [<at 0 3>,<at 0.3 1>]}, YCurve {keys [<at 0 1>,<at 0.5 1>,<at 1 1>]}, ZCurve {keys [<at 0 1>,<at 1 1>]}}})
     int32:NiagaraInt32=$Particles.MeshAxis
    )) {module=ScaleMeshSize001}
   &UpdateMeshOrientation(
    Rotation Coordinate Space=::NewEnumerator1
    Rotation Rate=&FloatFromCurve(
     FloatCurve=DataInterface() {propertyValues {Curve {keys [
         {
          arriveTangent=-3.1094
          interpMode=RCIM_Cubic
          leaveTangent=-3.1094
          leaveTangentWeight=1.0887
          tangentMode=RCTM_User
          tangentWeightMode=RCTWM_WeightedBoth
          time=0
          value=1
         }
         {
          arriveTangent=-0.7507
          arriveTangentWeight=0.4115
          interpMode=RCIM_Cubic
          leaveTangent=-0.7507
          leaveTangentWeight=0.4115
          tangentMode=RCTM_User
          tangentWeightMode=RCTWM_WeightedBoth
          time=1
          value=0
         }
        ]}}}
     Scale Curve=4
    )
    Rotation Vector=&CrossProduct(
     Vector A:Vector3f=&NormalizeVector(
      Vector To Normalize=&Subtract_Position(A=$Particles.Position, B=&SimulationPosition())
      Safe Normalize=true
      Safe Fallback Threshold=0
      Safe Fallback Vector=<0 0 0.001>
     )
     Vector B:Vector3f=<0 0 1>
     Normalize Incoming Vectors:NiagaraBool=true
    )
   )
   &DynamicMaterialParameters(
    Index 0 Param 1=&MakeFloatFromVector(Vector=$Particles.Scale, Channel=::NewEnumerator1)
    Index 0 Param 2=&RandomRangeFloat(Minimum=0, Maximum=1000)
    Index 0 Param 3=0.7
    Module.Param0WriteEnabled=true
    Module.Param0WriteEnabled1=true
    Module.Param1WriteEnabled=true
    Module.Param1WriteEnabled1=true
    Module.Param2WriteEnabled=true
    Module.Param2WriteEnabled1=true
    Module.Param3WriteEnabled=true
    Module.Param3WriteEnabled1=true
    Write Parameter Index 1=true
   ) {module=DynamicMaterialParameters001}
   &Drag(Drag=4, Use Linear Drag=true, Use Rotational Drag=false)
   &GravityForce(Gravity=&Multiply_VectorByFloat(Vector=<0 0 -980>, Float=0.1))
   &SolveForcesAndVelocity(Clamp Velocity=false, Limit Acceleration=false)
   &ScaleColor(
    Scale Alpha=&FloatFromCurve(FloatCurve=DataInterface() {propertyValues {Curve {keys [<at 0.75 1>,<at 1 0>]}}})
    ScaleA=true
    ScaleRGB=true
   ) {enabled=false}
   &ScaleColor(
    Scale Alpha=&FloatFromCurve(FloatCurve=DataInterface() {propertyValues {Curve {keys [{arriveTangent=-3.9434, interpMode=RCIM_Cubic, leaveTangent=-3.9434, tangentMode=RCTM_User, time=0.2, value=1}, <at 1 0 RCIM_Cubic>]}}})
    ScaleA=true
    ScaleRGB=true
   ) {module=ScaleColor001}
   &ScaleSpriteSize(
    Uniform Curve Scale=1.1
    Uniform Curve Sprite Scale=DataInterface() {propertyValues {Curve {keys [
        {arriveTangent=3.1805, interpMode=RCIM_Cubic, leaveTangent=3.1805, leaveTangentWeight=0.3333, tangentMode=RCTM_User, time=0, value=0.1}
        {
         arriveTangent=1.625
         arriveTangentWeight=0.05
         interpMode=RCIM_Cubic
         leaveTangent=1.625
         leaveTangentWeight=0.2833
         tangentMode=RCTM_User
         time=0.15
         value=0.65
        }
        {arriveTangentWeight=0.2167, interpMode=RCIM_Cubic, leaveTangentWeight=0.0667, tangentMode=RCTM_User, time=0.8, value=1}
       ]}}}
   )
   /SetVariables(Particles.Sprite color:LinearColor=&MultiplyLinearColor(
     Linear Color=&ColorFromCurve(ColorCurve=DataInterface() {propertyValues {
        AlphaCurve {keys [<at 0 0>, <at 0.1337 1>, {interpMode=RCIM_Cubic, leaveTangent=-1.1543, time=0.4608, value=1}, <at 0.8793 0 RCIM_Cubic>]}
        BlueCurve {keys [<at 0 0.4538 RCIM_Cubic>]}
        GreenCurve {keys [<at 0 0.3615 RCIM_Cubic>]}
        RedCurve {keys [<at 0 0.5 RCIM_Cubic>]}
       }})
     Scale Factor=&MakeFloat4FromVectorAndFloat("Vector (RGB)"=<1 1 1>, "Float (Alpha)"=0.1)
    )) {isSetParameters=true, module=SetVariables_#27}
  ]
  rendererClasses [&NiagaraMeshRendererProperties, &NiagaraSpriteRendererProperties]
  renderers [
   {rendererClass=&NiagaraMeshRendererProperties, rendererIndex=0}
   {rendererClass=&NiagaraMeshRendererProperties, rendererIndex=1}
   {rendererClass=&NiagaraSpriteRendererProperties, rendererIndex=2}
   {rendererClass=&NiagaraSpriteRendererProperties, rendererIndex=3}
  ]
 }
 SmokeGroundRings {
  emitterUpdateScript [
   &EmitterState()
   &SpawnBurst_Instantaneous(Use Spawn Probability=false, UseLoopCountLimit=false)
  ]
  particleSpawnScript [
   /Niagara/Modules/Spawn/Initialization/InitializeParticle(
    Color=<rgba 0.35 0.2778 0.21 100>
    Lifetime=&RandomRangeFloat(Minimum=0.95, Maximum=1.1)
    Module.Write Color=true
    Module.Write Lifetime=true
    Module.Write Mass=true
    Module.Write Material Random=true
    Module.Write Position=false
    Module.Write Scale=false
    Module.Write SpriteRotation=true
    Module.Write SpriteSize=true
    Position=/Niagara/Functions/Localspace/SimulationPosition()
    Sprite Rotation=&UniformRangedFloat(Minimum=-180, Maximum=180)
    Sprite Size=&Vector2DFromFloat(Value=5)
   ) {enabled=false}
   /Niagara/Modules/Spawn/Initialization/V2/InitializeParticle(
    Color=<rgba 0.7 0.5061 0.7 1>
    Lifetime=&RandomRangeFloat(Minimum=0.5, Maximum=0.5)
    Mass Mode=::NewEnumerator1
    Mesh Scale Mode=::NewEnumerator3
    Sprite Rotation Mode=::NewEnumerator2
    Sprite Size=&Vector2DFromFloat(Value=11)
    Sprite Size Mode=::NewEnumerator4
    UsePositionOffset=false
    Write Material Random=true
   ) {module=InitializeParticle001}
   /SetVariables(Particles.SpriteFacing=<0 0 1>) {isSetParameters=true}
   &SystemLocation(MaskSpawnGroup=false, Offset=<0 0 5>)
   &SpriteFacingAndAlignment(Sprite Alignment=<1 1 0>, Sprite Facing=<0 0 1>, WriteAlignment=true, WriteFacing=true)
  ]
  particleUpdateScript [
   &ParticleState()
   &DynamicMaterialParameters(
    Index 0 Param 1=&FloatFromCurve(FloatCurve=DataInterface() {propertyValues {Curve {keys [<at 0 0.8>,<at 1 2>]}}})
    Module.Param0WriteEnabled=true
    Module.Param1WriteEnabled=false
    Module.Param2WriteEnabled=false
    Module.Param3WriteEnabled=false
   )
   &ScaleSpriteSize(Scale Factor=&Vector2DFromFloat(Value=&FloatFromCurve(
      FloatCurve=DataInterface() {propertyValues {Curve {keys [{arriveTangent=4.9252, time=0, value=1}, {interpMode=RCIM_Cubic, leaveTangent=0.6903, time=1, value=1}]}}}
      Scale Curve=45
     )))
   &CameraOffset(Camera Offset Amount=20)
   &ScaleColor(
    Scale Alpha=&FloatFromCurve(
     FloatCurve=DataInterface() {propertyValues {Curve {keys [<at 0 1 RCIM_Cubic>, {interpMode=RCIM_Cubic, leaveTangent=-4.9981, time=1.0011, value=0.002}]}}}
     Scale Curve=2.5
    )
    ScaleA=true
    ScaleRGB=false
   )
   &InheritSourceMovement(Applied Position Scale=<0.5 0.5 0.5>, ApplyForce=false, ApplyPosition=true, ApplyPositionFalloff=false)
   &SolveForcesAndVelocity(Clamp Velocity=false, Limit Acceleration=false)
  ]
  rendererClasses [&NiagaraSpriteRendererProperties]
  renderers [{rendererClass=&NiagaraSpriteRendererProperties}]
 }
}
summary {
 systemName=NS_WeaveHarbor_Emote_Impact
 emitters [
  {
   emitterName=LightningAround
   rendererClasses [
    &NiagaraRibbonRendererProperties
    &NiagaraSpriteRendererProperties
    &NiagaraLightRendererProperties
   ]
  }
  {
   emitterName=LightningOut_EpicCin
   rendererClasses [
    &NiagaraRibbonRendererProperties
    &NiagaraSpriteRendererProperties
    &NiagaraLightRendererProperties
   ]
  }
  {emitterName=SmokeGroundRings, rendererClasses [&NiagaraSpriteRendererProperties]}
  {
   bEnabled=false
   emitterName=Poof001
   rendererClasses [&NiagaraMeshRendererProperties, &NiagaraSpriteRendererProperties]
  }
  {emitterName=Poof, rendererClasses [&NiagaraMeshRendererProperties, &NiagaraSpriteRendererProperties]}
  {
   bEnabled=false
   emitterName=LightningAround001
   rendererClasses [
    &NiagaraRibbonRendererProperties
    &NiagaraSpriteRendererProperties
    &NiagaraLightRendererProperties
   ]
  }
  {
   bEnabled=false
   emitterName=LightningOut001
   rendererClasses [
    &NiagaraRibbonRendererProperties
    &NiagaraSpriteRendererProperties
    &NiagaraLightRendererProperties
   ]
  }
 ]
 userVariables [
  User.bIsFrontEnd {type=NiagaraBool}
  User.bIsFrontEndPreview {type=NiagaraBool}
  User.bIsInWater {type=NiagaraBool}
  User.ForwardDot {type=NiagaraFloat, defaultValue=0, description=""}
  User.IsFrontEnd {type=NiagaraBool, defaultValue=true}
 ]
}
