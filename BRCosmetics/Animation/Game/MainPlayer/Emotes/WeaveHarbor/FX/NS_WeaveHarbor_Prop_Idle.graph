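# Niagara system for the idle glow on the WeaveHarbor emote prop, built from one active emitter and a disabled Debug emitter.
# Glow bursts ten pinkish-red sprites along a cylinder arc, facing outward and sized by scale sampled from a skeletal mesh.
# Brightness flares on start then settles, alpha follows time of day, and User.NotifyTime fades it out near 1.
# Three sprite renderers draw the particles; a view-angle alpha fade is present but disabled.
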NiagaraSystem /BRCosmetics/Animation/Game/MainPlayer/Emotes/WeaveHarbor/FX/NS_WeaveHarbor_Prop_Idle
dependencies {
 usedDataInterfaces [
  &NiagaraDataInterfacePlatformSet
  &NiagaraDataInterfaceColorCurve
  &NiagaraDataInterfaceCurve
  &NiagaraDataInterfaceCamera
  &NiagaraDataInterfaceSkeletalMesh
 ]
 usedDynamicInputs []
 usedModules [
  &self:SystemSpawnScript.SystemScriptSource.SystemScriptGraph.NiagaraNodeAssignment_0.SetVariables_#1
  &SystemState
 ]
}
emitters {
 Debug {
  bEnabled=false
  emitterUpdateScript [
   &EmitterState()
   &SpawnBurst_Instantaneous(Use Spawn Probability=false, UseLoopCountLimit=true)
  ]
  particleSpawnScript [&InitializeParticle(Color Mode=::NewEnumerator0, Lifetime=10, Position Offset=<0 0 200>, UsePositionOffset=true, Write Material Random=true)]
  particleUpdateScript [
   &ParticleState()
   &DynamicMaterialParameters(
    Index 0 Param 1=&FloatFromCurve(
     FloatCurve=DataInterface() {propertyValues {Curve {keys [<at -1 -1>,<at 1 1>]}}}
     CurveIndex=&DotProduct(
      Vector A=&ReturnCameraVector(Camera Vectors=::NewEnumerator1, Coordinate Space=::World)
      Vector B=&TransformVector(Vector=<0 0 1>, Source Space=::World, Destination Space=::Simulation)
     )
    )
    Module.Index0Param0Write=true
    Module.Index0Param1Write=true
    Module.Index0Param2Write=true
    Module.Index0Param3Write=true
   )
  ]
  rendererClasses [&NiagaraSpriteRendererProperties]
  renderers [{rendererClass=&NiagaraSpriteRendererProperties}]
 }
 Glow {
  emitterSpawnScript [/SetVariables(Emitter.NumLineParticles:NiagaraInt32=&MakeCustomIntFromBool(
     Bool=&IsPlatformSetActive(New Platform Set=DataInterface() {propertyValues {Platforms {qualityLevelMask=-4}}})
     True INT=10
     False INT=10
    )) {isSetParameters=true}]
  emitterUpdateScript [
   &EmitterState()
   &SpawnBurst_Instantaneous(Spawn Count=$Emitter.NumLineParticles, Use Spawn Probability=false, UseLoopCountLimit=true)
  ]
  particleSpawnScript [
   &PartitionParticles(
    Normalized Index In Partition Includes One=true
    Number Of Partitions=1
    Particles Per Partition:NiagaraInt32=$Emitter.NumLineParticles
    Partition Mode:ENiagara_PartitionMode=::NewEnumerator1
   )
   &SkeletalMeshLocation(Bone Sampling Mode=::NewEnumerator4, MaskSpawnGroup=false, Skeletal Mesh=$System.Skeletal Mesh)
   &InitializeParticle(
    Color=&LerpLinearColors(
     StartColor=&ColorFromCurve(
      ColorCurve=DataInterface() {propertyValues {
        AlphaCurve {keys [<at 0 1 RCIM_Cubic>]}
        BlueCurve {keys [<at 0.2732 0.04>,<at 0.6542 0.04>]}
        GreenCurve {keys [<at 0.2732 0>,<at 0.6542 0>]}
        RedCurve {keys [<at 0.2732 1>,<at 0.6542 1>]}
       }}
      CurveIndex=$Particles.PartitionParticles.Normalized Index In Partition
     )
     EndColor=<rgba 1 1 1 1>
     LerpFactor=0.3
    )
    Lifetime=5
    Sprite Size=&Multiply_Vector2D_ByFloat(
     Float=&MakeFloatFromVector(Vector=$Particles.SkeletalMeshLocation.SampledScale, Channel=::NewEnumerator0)
     Vector2D=&Multiply_Vector2D_ByFloat(Float=0.624, Vector2D=<50 50>)
    )
    Sprite Size Mode=::NewEnumerator4
    UsePositionOffset=false
    Write Material Random=true
   )
   &self:ScratchModule() {module=ScratchModule}
   &ShapeLocation(
    Cylinder Distribution=::NewEnumerator2
    Cylinder Height=0
    Cylinder Radius=52
    MaskSpawnGroup=false
    Offset=<0 0 13.5>
    Offset Mode=::NewEnumerator0
    Shape Primitive=::NewEnumerator1
    U Position=&Add_Float(
     A=&Add_Float(
      A=&Divide_Float(A=1, B=12)
      B=&Multiply_Float(A=1, B=&Divide_Float(A=3, B=6))
     )
     B=&Multiply_Float(A=0.12, B=&Lerp_Float(A=-1, B=1, Alpha=$Particles.PartitionParticles.Normalized Index In Partition))
    )
    UseOwnerScale=true
   )
   &SpriteFacingAndAlignment(Sprite Facing=$Particles.ShapeLocation.ShapeNormal, WriteAlignment=false, WriteFacing=true)
   &CameraOffset(Camera Offset Amount=10)
   /SetVariables(Particles.otherPosition:NiagaraPosition=&AddVectorToPosition(
     Position=$Particles.Position
     Vector=&Multiply_Vector(
      A=Linked() {linkedVariable {name=Particles.ShapeLocation.ShapeNormal, type=NiagaraPosition}}
      B=&VectorFromFloat(Value=1)
     )
    )) {isSetParameters=true}
  ]
  particleUpdateScript [
   &ParticleState(Kill Particles When Lifetime Has Elapsed=false, Let Infinitely Lived Particles Die When Emitter Deactivates=true)
   &ScaleSpriteSize(Uniform Curve Index=&MakeFloatFromVector(Vector=$Particles.SkeletalMeshLocation.SampledScale, Channel=::NewEnumerator0))
   &ScaleColor(
    Scale Alpha=&FNTimeOfDayScalar(Night Scale=0.025, Morning Scale=0.03, Day Scale=0.03, Evening Scale=0.03)
    Scale RGB=&VectorFromFloat(Value=&FloatFromCurve(
      FloatCurve=DataInterface() {propertyValues {Curve {keys [
          {arriveTangent=34.5463, interpMode=RCIM_Cubic, leaveTangent=34.5463, tangentMode=RCTM_User, time=0, value=0}
          {arriveTangentWeight=0.1795, interpMode=RCIM_Cubic, leaveTangentWeight=0.1795, tangentMode=RCTM_User, time=0.2, value=3}
          {
           arriveTangent=-1.9645
           arriveTangentWeight=0.1795
           interpMode=RCIM_Cubic
           leaveTangent=-1.9645
           leaveTangentWeight=0.5385
           tangentMode=RCTM_User
           time=0.4
           value=1
          }
          <at 1 1 RCIM_Cubic>
         ]}}}
      CurveIndex=$Emitter.Age
      Scale Curve=&Multiply_Float(
       A=2
       B=&MakeFloatFromVector(Vector=$Particles.SkeletalMeshLocation.SampledScale, Channel=::NewEnumerator0)
      )
     ))
    ScaleA=true
    ScaleRGB=true
   )
   &ScaleColor(
    Scale Alpha=&FloatFromCurve(FloatCurve=DataInterface() {propertyValues {Curve {keys [<at 0.9 1>,<at 1 0>]}}}, CurveIndex=$User.NotifyTime)
    ScaleA=true
    ScaleRGB=true
   ) {module=ScaleColor001}
   &ScaleColor(Scale RGB=&VectorFromFloat(Value=0.5), ScaleA=true, ScaleRGB=true) {module=ScaleColor002}
   &ScaleColor(
    Scale Alpha=&ClampFloat(
     Float=&Add_Float(
      A=&Lerp_Float(
       A=0
       B=2
       Alpha=&FloatFromCurve(
        FloatCurve=DataInterface() {propertyValues {Curve {keys [<at -1 0 RCIM_Cubic>, {arriveTangent=1.2685, interpMode=RCIM_Cubic, leaveTangent=1.2685, tangentMode=RCTM_User, time=1, value=1}]}}}
        CurveIndex=&DotProduct(
         Vector A=&NormalizeVector(
          Vector To Normalize=&Multiply_Vector(
           A=&ReturnCameraVector(Camera Vectors=::NewEnumerator4, Coordinate Space=::Simulation, Origin Position=$Particles.Position)
           B=<1 1 0>
          )
          Safe Normalize=true
          Safe Fallback Threshold=0
          Safe Fallback Vector=<0 0 0.001>
         )
         Vector B=&NormalizeVector(
          Vector To Normalize=&Multiply_Vector(A=&Subtract_Position(A=$Particles.Position, B=/Niagara/DynamicInputs/Helpers/SimulationPosition()), B=<1 1 1>)
          Safe Normalize=true
          Safe Fallback Threshold=0
          Safe Fallback Vector=<0 0 0.001>
         )
        )
       )
      )
      B=&FloatFromCurve(
       FloatCurve=DataInterface() {propertyValues {Curve {keys [<at -1 1 RCIM_Cubic>, {arriveTangent=-17.8878, interpMode=RCIM_Cubic, leaveTangent=-17.8878, tangentMode=RCTM_User, time=-0.9, value=0.5}]}}}
       CurveIndex=&DotProduct(
        Vector A=&ReturnCameraVector(Camera Vectors=::NewEnumerator1, Coordinate Space=::World)
        Vector B=&TransformVector(Vector=<0 0 1>, Source Space=::World, Destination Space=::Simulation)
       )
       Scale Curve=0
      )
     )
     Max=2
    )
    ScaleA=true
    ScaleRGB=true
   ) {enabled=false, module=ScaleColor003}
   &DynamicMaterialParameters(
    Index 0 Param 1=&MakeFloatFromLinearColor(LinearColor=$Particles.Color, Channel:ENiagaraLinearColor_Channels=::NewEnumerator3)
    Module.Index0Param0Write=true
    Module.Index0Param1Write=true
    Module.Index0Param2Write=true
    Module.Index0Param3Write=true
   ) {enabled=false}
   /SetVariables(
    Particles.PLate Size:Vector2f=&Multiply_Vector2D(A=<1 1.25>, B=$Particles.SpriteSize)
    Particles.Plate color:LinearColor=&MultiplyLinearColor(
     Linear Color=$Particles.Color
     Scale Factor=&MakeFloat4FromVectorAndFloat("Vector (RGB)"=&VectorFromFloat(Value=7), "Float (Alpha)"=1)
    )
   ) {isSetParameters=true}
  ]
  rendererClasses [&NiagaraSpriteRendererProperties]
  renderers [
   {rendererClass=&NiagaraSpriteRendererProperties}
   {rendererClass=&NiagaraSpriteRendererProperties, rendererIndex=1}
   {rendererClass=&NiagaraSpriteRendererProperties, rendererIndex=2}
  ]
 }
}
summary {
 systemName=NS_WeaveHarbor_Prop_Idle
 emitters [
  {emitterName=Glow, rendererClasses [&NiagaraSpriteRendererProperties]}
  {bEnabled=false, emitterName=Debug, rendererClasses [&NiagaraSpriteRendererProperties]}
 ]
 userVariables [User.NotifyTime {type=NiagaraFloat, defaultValue=0, description=""}]
}
